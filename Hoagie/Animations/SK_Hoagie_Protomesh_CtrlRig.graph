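# Control Rig for the Hoagie helicopter's prototype mesh that constrains the seven blade_flex_pos bones of the center main-rotor blade to vertical movement.
# One sequential execution chain handles each bone in local space, taking X and Y translation from its initial pose and keeping its current Z.
# Each write also sets identity rotation and unit scale, without propagating to children.
# The rest of the imported skeleton, including control rods, IK bones and boosters, is untouched by the graph.
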ControlRigBlueprint /Hoagie/Animations/SK_Hoagie_Protomesh_CtrlRig
graphs [
 RigVMFunctionLibrary {links [], nodes []}
 RigVMModel {nodes [
   n1 = &RigUnit_BeginExecution()
   n2 = /Script/ControlRig.RigUnit_GetInitialBoneTransform(Bone=blade_flex_pos_1_c, Space=LocalSpace) {is=struct RigVMUnitNode}
   /Script/ControlRig.RigUnit_SetBoneTransform(
    Bone=blade_flex_pos_1_c
    Space=LocalSpace
    Transform=(Rotation=(X=0,Y=0,Z=0,W=1),Translation=(X=0,Y=0,Z=0),Scale3D=(X=1,Y=1,Z=1))
    Weight=1
    bPropagateToChildren=False
    ExecutePin=@n1.ExecutePin
    Transform.Translation.Z=&RigUnit_GetBoneTransform(Bone=blade_flex_pos_1_c, Space=LocalSpace).Transform.Translation.Z
    Transform.Translation.X=@n2.Transform.Translation.X
    Transform.Translation.Y=@n2.Transform.Translation.Y
   ) {is=struct RigVMUnitNode}
   n3 = &RigUnit_GetInitialBoneTransform(Bone=blade_flex_pos_2_c, Space=LocalSpace)
   > &RigUnit_SetBoneTransform(
    Bone=blade_flex_pos_2_c
    Space=LocalSpace
    Transform=(Rotation=(X=0,Y=0,Z=0,W=1),Translation=(X=0,Y=0,Z=0),Scale3D=(X=1,Y=1,Z=1))
    Weight=1
    bPropagateToChildren=False
    Transform.Translation.Z=&RigUnit_GetBoneTransform(Bone=blade_flex_pos_2_c, Space=LocalSpace).Transform.Translation.Z
    Transform.Translation.X=@n3.Transform.Translation.X
    Transform.Translation.Y=@n3.Transform.Translation.Y
   )
   n4 = &RigUnit_GetInitialBoneTransform(Bone=blade_flex_pos_3_c, Space=LocalSpace)
   > &RigUnit_SetBoneTransform(
    Bone=blade_flex_pos_3_c
    Space=LocalSpace
    Transform=(Rotation=(X=0,Y=0,Z=0,W=1),Translation=(X=0,Y=0,Z=0),Scale3D=(X=1,Y=1,Z=1))
    Weight=1
    bPropagateToChildren=False
    Transform.Translation.Z=&RigUnit_GetBoneTransform(Bone=blade_flex_pos_3_c, Space=LocalSpace).Transform.Translation.Z
    Transform.Translation.X=@n4.Transform.Translation.X
    Transform.Translation.Y=@n4.Transform.Translation.Y
   )
   n5 = &RigUnit_GetInitialBoneTransform(Bone=blade_flex_pos_4_c, Space=LocalSpace)
   > &RigUnit_SetBoneTransform(
    Bone=blade_flex_pos_4_c
    Space=LocalSpace
    Transform=(Rotation=(X=0,Y=0,Z=0,W=1),Translation=(X=0,Y=0,Z=0),Scale3D=(X=1,Y=1,Z=1))
    Weight=1
    bPropagateToChildren=False
    Transform.Translation.Z=&RigUnit_GetBoneTransform(Bone=blade_flex_pos_4_c, Space=LocalSpace).Transform.Translation.Z
    Transform.Translation.X=@n5.Transform.Translation.X
    Transform.Translation.Y=@n5.Transform.Translation.Y
   )
   n6 = &RigUnit_GetInitialBoneTransform(Bone=blade_flex_pos_5_c, Space=LocalSpace)
   > &RigUnit_SetBoneTransform(
    Bone=blade_flex_pos_5_c
    Space=LocalSpace
    Transform=(Rotation=(X=0,Y=0,Z=0,W=1),Translation=(X=0,Y=0,Z=0),Scale3D=(X=1,Y=1,Z=1))
    Weight=1
    bPropagateToChildren=False
    Transform.Translation.Z=&RigUnit_GetBoneTransform(Bone=blade_flex_pos_5_c, Space=LocalSpace).Transform.Translation.Z
    Transform.Translation.X=@n6.Transform.Translation.X
    Transform.Translation.Y=@n6.Transform.Translation.Y
   )
   n7 = &RigUnit_GetInitialBoneTransform(Bone=blade_flex_pos_6_c, Space=LocalSpace)
   > &RigUnit_SetBoneTransform(
    Bone=blade_flex_pos_6_c
    Space=LocalSpace
    Transform=(Rotation=(X=0,Y=0,Z=0,W=1),Translation=(X=0,Y=0,Z=0),Scale3D=(X=1,Y=1,Z=1))
    Weight=1
    bPropagateToChildren=False
    Transform.Translation.Z=&RigUnit_GetBoneTransform(Bone=blade_flex_pos_6_c, Space=LocalSpace).Transform.Translation.Z
    Transform.Translation.X=@n7.Transform.Translation.X
    Transform.Translation.Y=@n7.Transform.Translation.Y
   )
   n8 = &RigUnit_GetInitialBoneTransform(Bone=blade_flex_pos_7_c, Space=LocalSpace)
   > &RigUnit_SetBoneTransform(
    Bone=blade_flex_pos_7_c
    Space=LocalSpace
    Transform=(Rotation=(X=0,Y=0,Z=0,W=1),Translation=(X=0,Y=0,Z=0),Scale3D=(X=1,Y=1,Z=1))
    Weight=1
    bPropagateToChildren=False
    Transform.Translation.Z=&RigUnit_GetBoneTransform(Bone=blade_flex_pos_7_c, Space=LocalSpace).Transform.Translation.Z
    Transform.Translation.X=@n8.Transform.Translation.X
    Transform.Translation.Y=@n8.Transform.Translation.Y
   )
  ]}
]
hierarchy [
 Hoagie_root {loc=<0 0 0>, rot=<0 0 0 1>}
 body {parent=Hoagie_root, loc=<-93.1067 -0.0001 186.2111>, rot=<0 0 0 1>}
 Hoagie_tailrotor {parent=body, loc=<-593.6887 0.0001 83.4423>, rot=<0 0 0 1>}
 Hoagie_mainrotor {parent=body, loc=<93.1067 0.0001 133.7889>, rot=<0 0 0 1>}
 swashplate_top {parent=Hoagie_mainrotor, loc=<0 0 24.2521>, rot=<0 0 0 1>}
 controlrod_r_tgt {parent=swashplate_top, loc=<-6.6279 37.5888 0>, rot=<-0.021 -0.0784 0.258 -0.9627>}
 controlrod_l_tgt {parent=swashplate_top, loc=<35.8667 -13.0542 0>, rot=<0.0056 -0.021 -0.2588 -0.9657>}
 controlrod_c_tgt {parent=swashplate_top, loc=<-29.2388 -24.5341 0>, rot=<-0.0355 0 0 -0.9994>}
 blade_l {parent=Hoagie_mainrotor, loc=<14.864 -25.745 52.7175>, rot=<-0.6736 -0.215 -0.1507 -0.6909>}
 controlrod_l {parent=blade_l, loc=<24.6434 0 0>, rot=<0 -0.0126 0.6729 -0.7396>}
 controlrod_tip_l {parent=controlrod_l, loc=<26.1529 0 0>, rot=<0 0 0 1>}
 blade_flex_1_l {parent=blade_l, loc=<0.0001 0 30.8619>, rot=<0 0 0 1>}
 blade_flex_2_l {parent=blade_flex_1_l, loc=<0 0 200.7743>, rot=<0 0 0 1>}
 blade_flex_3_l {parent=blade_flex_2_l, loc=<0 0 231.9171>, rot=<0 0 0 1>}
 blade_r {parent=Hoagie_mainrotor, loc=<14.8638 25.7451 52.7175>, rot=<-0.1507 -0.6909 -0.6736 -0.215>}
 controlrod_r {parent=blade_r, loc=<24.6434 0 0>, rot=<0 -0.0126 0.6729 -0.7396>}
 controlrod_tip_r {parent=controlrod_r, loc=<26.1529 0 0>, rot=<0 0 0 -1>}
 blade_flex_1_r {parent=blade_r, loc=<0.0001 0 30.8619>, rot=<0 0 0 1>}
 blade_flex_2_r {parent=blade_flex_1_r, loc=<0 0 200.7744>, rot=<0 0 0 1>}
 blade_flex_3_r {parent=blade_flex_2_r, loc=<0 0 231.9171>, rot=<0 0 0 1>}
 blade_c {parent=Hoagie_mainrotor, loc=<-29.7279 0.0001 52.7175>, rot=<0.5232 -0.4756 -0.5232 0.4756>}
 controlrod_c {parent=blade_c, loc=<24.3561 0 -0.1558>, rot=<0 -0.0086 0.6685 -0.7437>}
 controlrod_tip_c {parent=controlrod_c, loc=<26.152 0 0>, rot=<0 0 0 1>}
 blade_flex_1_c {parent=blade_c, loc=<0.0001 0 30.8619>, rot=<0 0 0 1>}
 blade_flex_2_c {parent=blade_flex_1_c, loc=<0 0 200.7743>, rot=<0 0 0 1>}
 blade_flex_3_c {parent=blade_flex_2_c, loc=<0 0 231.9171>, rot=<0 0 0 1>}
 blade_ik_r {parent=Hoagie_mainrotor, loc=<14.8638 25.7451 52.7175>, rot=<-0.153 -0.6904 -0.6744 -0.2127>}
 controlrod_ik_r {parent=blade_ik_r, loc=<24.6429 -0.1662 0>, rot=<0 0 0.6755 -0.7373>}
 controlrod_ik_tip_r {parent=controlrod_ik_r, loc=<26.1483 0 -0.4889>, rot=<0 0 0 -1>}
 blade_ik_c {parent=Hoagie_mainrotor, loc=<-29.7279 0.0001 52.7175>, rot=<0.5214 -0.4777 -0.5214 0.4777>}
 controlrod_ik_c {parent=blade_ik_c, loc=<24.3554 -0.1914 -0.1558>, rot=<-0.0261 0.024 0.6751 -0.7369>}
 controlrod_tip_ik_c {parent=controlrod_ik_c, loc=<26.1483 0.2642 -0.3529>, rot=<0 0 0 1>}
 blade_ik_l {parent=Hoagie_mainrotor, loc=<14.864 -25.745 52.7175>, rot=<-0.6744 -0.2127 -0.153 -0.6904>}
 controlrod_ik_l {parent=blade_ik_l, loc=<24.6428 -0.1662 0>, rot=<0 0 0.6755 -0.7373>}
 controlrod_tip_ik_l {parent=controlrod_ik_l, loc=<26.1483 0 -0.4892>, rot=<0 0 0 -1>}
 blade_flex_pos_1_c {parent=Hoagie_mainrotor, loc=<-147.2968 0.4341 56.1989>, rot=<0 0 0 1>}
 blade_flex_pos_2_c {parent=Hoagie_mainrotor, loc=<-235.5405 0.6936 56.371>, rot=<0 0 0 1>}
 blade_flex_pos_3_c {parent=Hoagie_mainrotor, loc=<-323.7843 0.9532 56.5429>, rot=<0 0 0 1>}
 blade_flex_pos_4_c {parent=Hoagie_mainrotor, loc=<-412.028 1.2128 56.7149>, rot=<0 0 0 1>}
 blade_flex_pos_5_c {parent=Hoagie_mainrotor, loc=<-500.2719 1.4724 56.8869>, rot=<0 0 0 1>}
 blade_flex_pos_6_c {parent=Hoagie_mainrotor, loc=<-588.5156 1.732 57.0589>, rot=<0 0 0 1>}
 blade_flex_pos_7_c {parent=Hoagie_mainrotor, loc=<-604.5 1.7779 52.5125>, rot=<0 0 0 1>}
 nozzle_1_r {parent=body, loc=<-209.8984 19.9212 -12.0815>, rot=<-0.5 -0.5 0.5 0.5>}
 nozzle_1_r_end {parent=nozzle_1_r, loc=<43.9305 0 0>, rot=<0 0 0 -1>}
 booster_r {parent=nozzle_1_r_end, loc=<28.1039 -0.1271 -1.7372>, rot=<0.5 0.5 0.5 -0.5>}
 booster_fan_r {parent=booster_r, loc=<-35.7356 -0.0467 0.0002>, rot=<0 0 0 1>}
 booster_coolified_in_r {parent=booster_r, loc=<27.0908 0 0>, rot=<0 0 0 1>}
 booster_coolified_out_r {parent=booster_r, loc=<37.0138 0 0>, rot=<0 0 0 1>}
 booster_01_r {parent=booster_r, loc=<28.6827 0 -22.0114>, rot=<0 0 1 0>}
 booster_02_r {parent=booster_r, loc=<28.6827 -14.1484 -16.8617>, rot=<0 0.342 0.9397 0>}
 booster_03_r {parent=booster_r, loc=<28.6827 -21.6767 -3.8224>, rot=<0 0.6428 0.766 0>}
 booster_04_r {parent=booster_r, loc=<28.6827 -19.0622 11.0056>, rot=<0 -0.866 -0.5 0>}
 booster_05_r {parent=booster_r, loc=<28.6827 -7.5281 20.6837>, rot=<0 -0.9848 -0.1736 0>}
 booster_06_r {parent=booster_r, loc=<28.6827 7.5285 20.6837>, rot=<0 -0.9848 0.1736 0>}
 booster_07_r {parent=booster_r, loc=<28.6827 19.0624 11.0057>, rot=<0 -0.866 0.5 0>}
 booster_08_r {parent=booster_r, loc=<28.6827 21.6771 -3.8225>, rot=<0 -0.6428 0.766 0>}
 booster_09_r {parent=booster_r, loc=<28.6827 14.1488 -16.8623>, rot=<0 -0.342 0.9397 0>}
 nozzle_1_l {parent=body, loc=<-209.8984 -19.9209 -12.0815>, rot=<0.5 0.5 -0.5 -0.5>}
 nozzle_1_l_end {parent=nozzle_1_l, loc=<-43.9305 0 0>, rot=<0 0 0 -1>}
 booster_l {parent=nozzle_1_l_end, loc=<-28.1039 -0.1271 -1.7372>, rot=<-0.5 -0.5 0.5 -0.5>}
 booster_01_l {parent=booster_l, loc=<-28.6827 0 22.0114>, rot=<1 0 0 0>}
 booster_02_l {parent=booster_l, loc=<-28.6827 14.1483 16.8617>, rot=<0.9397 0 0 0.342>}
 booster_03_l {parent=booster_l, loc=<-28.6827 21.6767 3.8224>, rot=<-0.766 0 0 -0.6428>}
 booster_04_l {parent=booster_l, loc=<-28.6827 19.0622 -11.0056>, rot=<-0.5 0 0 -0.866>}
 booster_05_l {parent=booster_l, loc=<-28.6827 7.5281 -20.6837>, rot=<-0.1736 0 0 -0.9848>}
 booster_06_l {parent=booster_l, loc=<-28.6827 -7.5285 -20.6837>, rot=<0.1736 0 0 -0.9848>}
 booster_07_l {parent=booster_l, loc=<-28.6827 -19.0624 -11.0057>, rot=<0.5 0 0 -0.866>}
 booster_08_l {parent=booster_l, loc=<-28.6827 -21.6771 3.8225>, rot=<0.766 0 0 -0.6428>}
 booster_09_l {parent=booster_l, loc=<-28.6827 -14.1488 16.8623>, rot=<0.9397 0 0 -0.342>}
 booster_coolified_in_l {parent=booster_l, loc=<-27.0908 0 0>, rot=<0 0 0 1>}
 booster_coolified_out_l {parent=booster_l, loc=<-37.0138 0 0>, rot=<0 0 0 1>}
 booster_fan_l {parent=booster_l, loc=<35.7357 0.0467 -0.0002>, rot=<0 0 0 1>}
 controlrod_l_fnt_tgt {parent=body, loc=<121.6352 -28.5455 94.4828>, rot=<0 0 0 -1>}
 controlrod_l_bck_tgt {parent=body, loc=<64.5612 -28.5283 94.4828>, rot=<0 0 0 -1>}
 controlrod_r_fnt_tgt {parent=body, loc=<121.6523 28.5286 94.4828>, rot=<0 0 0 -1>}
 controlrod_r_bck_tgt {parent=body, loc=<64.5783 28.5457 94.4828>, rot=<0 0 0 -1>}
 swashplate_bottom {parent=body, loc=<93.1067 0.0001 134.3223>, rot=<0 0 0 1>}
 controlrod_l_fnt {parent=swashplate_bottom, loc=<28.5285 -28.5455 2.8766>, rot=<0 -0.7071 0 -0.7071>}
 controlrod_l_fnt_tip {parent=controlrod_l_fnt, loc=<42.7162 0.0002 -0.0001>, rot=<0.488 0.3913 0.488 -0.6087>}
 controlrod_l_bck {parent=swashplate_bottom, loc=<-28.5456 -28.5284 2.8766>, rot=<0 -0.7071 0 -0.7071>}
 controlrod_l_bck_tip {parent=controlrod_l_bck, loc=<42.7162 0 0>, rot=<-0.5 0.5 -0.5 -0.5>}
 controlrod_r_fnt {parent=swashplate_bottom, loc=<28.5456 28.5285 2.8766>, rot=<0 -0.7071 0 -0.7071>}
 controlrod_r_fnt_tip {parent=controlrod_r_fnt, loc=<42.7162 0 0>, rot=<-0.5 0.5 -0.5 -0.5>}
 controlrod_r_bck {parent=swashplate_bottom, loc=<-28.5285 28.5456 2.8766>, rot=<0 -0.7071 0 -0.7071>}
 controlrod_r_bck_tip {parent=controlrod_r_bck, loc=<42.7162 0 0>, rot=<0.488 -0.6087 0.488 0.3913>}
 cable_l {parent=body, loc=<-132.4431 -14.4444 -54.878>, rot=<-0.1634 -0.6273 0.0463 -0.76>}
 cable_r {parent=body, loc=<-132.813 -1.0702 -54.6811>, rot=<-0.0158 -0.6845 -0.0833 -0.7241>}
 seat {parent=body, loc=<251.0573 0.0827 -64.7336>, rot=<0 0 0 1>}
 seat_bottom {parent=seat, loc=<0 0 0>, rot=<0 0.0059 0 -1>}
 collective_base {parent=seat_bottom, loc=<4.2662 -36.1817 -11.7488>, rot=<0 -0.0059 0 -1>}
 collective {parent=collective_base, loc=<3.1317 -0.0001 9.6594>, rot=<0 -0.1566 0 -0.9877>}
 seat_back {parent=seat, loc=<-25.8243 -0.0825 -1.8675>, rot=<0 0.0667 0 -0.9978>}
 seat_head {parent=seat_back, loc=<0.6375 0 75.4441>, rot=<0 -0.2006 0 -0.9797>}
 pedals {parent=body, loc=<316.1001 -0.1919 -88.285>, rot=<0 0 0 1>}
 control_lever {parent=pedals, loc=<-18.5028 0 21.6703>, rot=<0 0.1139 0 -0.9935>}
 lever_trigger {parent=control_lever, loc=<0.9676 0.1921 37.2355>, rot=<0 0.0419 0 -0.9991>}
 lever_button {parent=control_lever, loc=<-1.1408 0.1921 45.9466>, rot=<0 0.0194 0.0001 -0.9998>}
 skid_r {parent=body, loc=<73.8509 59.0793 -113.35>, rot=<0 0 0.0399 -0.9992>}
 platform_r {parent=skid_r, loc=<15.8541 36.2363 -8.4578>, rot=<0 0.0004 -0.0574 -0.9984>}
 skid_l {parent=body, loc=<73.8509 -59.0791 -113.35>, rot=<0 0 -0.0399 -0.9992>}
 platform_l {parent=skid_l, loc=<15.8646 -36.1052 -8.4578>, rot=<0.0078 -0.0001 0.0574 -0.9983>}
 seat_bottom_r {parent=body, loc=<-9.9816 36.6902 -52.2052>, rot=<0 -0.0381 0 -0.9993>}
 seat_bottom_l {parent=body, loc=<-9.9816 -36.69 -52.2052>, rot=<0 -0.0381 0 -0.9993>}
 seat_head_r {parent=body, loc=<-33.3292 36.6901 20.5957>, rot=<0 0 0 1>}
 seat_head_l {parent=body, loc=<-33.3292 -36.6899 20.5957>, rot=<0 0 0 1>}
 gauge_l {parent=body, loc=<308.7597 -31.044 -27.4586>, rot=<-0.0394 -0.1429 -0.0184 -0.9888>}
 gauge_c {parent=body, loc=<306.1646 -13.9595 -31.0737>, rot=<-0.0394 -0.1429 -0.0184 -0.9888>}
 gauge_r {parent=body, loc=<308.7597 3.3948 -29.7509>, rot=<-0.0394 -0.1429 -0.0184 -0.9888>}
 screen {parent=body, loc=<303.9903 40.1053 -35.2644>, rot=<0.1633 0.0225 0.986 0.0249>}
]
