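# Animation Blueprint for the Hoagie helicopter vehicle, animating cockpit controls, rotors, blades and boost engines.
# LeverMovement and EngineBoost state machines plus a collective blendspace are layered per bone, falling back to a static sequence without a driver.
# Bone modifiers spin rotors and fans, then a shutdown additive, weighted centre-blade flex constraints and a Control Rig finish the pose.
# The blade-hiding scale chain is unconnected, and the EventGraph is empty.
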AnimBlueprint /Hoagie/Animations/SK_Hoagie_Protomesh_AnimBP
target_skeleton=/Hoagie/Meshes/SK_Hoagie_Protomesh_Skeleton
graphs [
 AnimGraph {nodes [
   n1 = Get bShouldHideBlades() -> (bShouldHideBlades:bool)
   n2 = Get BladeScale() -> (BladeScale:Vector)
   n3 = "Transform (Modify) Bone - Bone: blade_flex_2_r"!(ComponentPose=Local To Component(), bAlphaBoolEnabled:bool=@n1, Scale=@n2) -> (Pose) {
    alphaInputType=Bool
    rotationSpace=BCS_BoneSpace
    scaleMode=BMM_Replace
    alphaBoolBlend {blendInTime=0.5, blendOption=Linear, blendOutTime=0.5}
    boneToModify {boneName=blade_flex_2_r}
   }
   n4 = "Transform (Modify) Bone - Bone: blade_flex_1_r"!(ComponentPose=@n3, bAlphaBoolEnabled:bool=@n1, Scale=@n2) -> (Pose) {
    alphaInputType=Bool
    rotationSpace=BCS_BoneSpace
    scaleMode=BMM_Replace
    alphaBoolBlend {blendInTime=0.5, blendOption=Linear, blendOutTime=0.5}
    boneToModify {boneName=blade_flex_1_r}
   }
   n5 = "Transform (Modify) Bone - Bone: blade_flex_3_r"!(ComponentPose=@n4, bAlphaBoolEnabled:bool=@n1, Scale=@n2) -> (Pose) {
    alphaInputType=Bool
    rotationSpace=BCS_BoneSpace
    scaleMode=BMM_Replace
    alphaBoolBlend {blendInTime=0.5, blendOption=Linear, blendOutTime=0.5}
    boneToModify {boneName=blade_flex_3_r}
   }
   n6 = "Transform (Modify) Bone - Bone: blade_flex_2_l"!(ComponentPose=@n5, bAlphaBoolEnabled:bool=@n1, Scale=@n2) -> (Pose) {
    alphaInputType=Bool
    rotationSpace=BCS_BoneSpace
    scaleMode=BMM_Replace
    alphaBoolBlend {blendInTime=0.5, blendOption=Linear, blendOutTime=0.5}
    boneToModify {boneName=blade_flex_2_l}
   }
   n7 = "Transform (Modify) Bone - Bone: blade_flex_1_l"!(ComponentPose=@n6, bAlphaBoolEnabled:bool=@n1, Scale=@n2) -> (Pose) {
    alphaInputType=Bool
    rotationSpace=BCS_BoneSpace
    scaleMode=BMM_Replace
    alphaBoolBlend {blendInTime=0.5, blendOption=Linear, blendOutTime=0.5}
    boneToModify {boneName=blade_flex_1_l}
   }
   n8 = "Transform (Modify) Bone - Bone: blade_flex_3_l"!(ComponentPose=@n7, bAlphaBoolEnabled:bool=@n1, Scale=@n2) -> (Pose) {
    alphaInputType=Bool
    rotationSpace=BCS_BoneSpace
    scaleMode=BMM_Replace
    alphaBoolBlend {blendInTime=0.5, blendOption=Linear, blendOutTime=0.5}
    boneToModify {boneName=blade_flex_3_l}
   }
   n9 = "Transform (Modify) Bone - Bone: blade_flex_2_c"!(ComponentPose=@n8, bAlphaBoolEnabled:bool=@n1, Scale=@n2) -> (Pose) {
    alphaInputType=Bool
    rotationSpace=BCS_BoneSpace
    scaleMode=BMM_Replace
    alphaBoolBlend {blendInTime=0.5, blendOption=Linear, blendOutTime=0.5}
    boneToModify {boneName=blade_flex_2_c}
   }
   n10 = "Transform (Modify) Bone - Bone: blade_flex_1_c"!(ComponentPose=@n9, bAlphaBoolEnabled:bool=@n1, Scale=@n2) -> (Pose) {
    alphaInputType=Bool
    rotationSpace=BCS_BoneSpace
    scaleMode=BMM_Replace
    alphaBoolBlend {blendInTime=0.5, blendOption=Linear, blendOutTime=0.5}
    boneToModify {boneName=blade_flex_1_c}
   }
   n11 = "Transform (Modify) Bone - Bone: blade_flex_3_c"!(ComponentPose=@n10, bAlphaBoolEnabled:bool=@n1, Scale=@n2) -> (Pose) {
    alphaInputType=Bool
    rotationSpace=BCS_BoneSpace
    scaleMode=BMM_Replace
    alphaBoolBlend {blendInTime=0.5, blendOption=Linear, blendOutTime=0.5}
    boneToModify {boneName=blade_flex_3_c}
   }
   Component To Local(ComponentPose=@n11)
   n12 = Get YawBlendSpaceInput() -> (YawBlendSpaceInput:float)
   n13 = Get EngineRotation() -> (EngineRotation:Rotator)
   n14 = "Transform (Modify) Bone - Bone: nozzle_1_l_end"!(
    ComponentPose=Local To Component(LocalPose=Blendspace Player 'Hoagie_Engine_Boosting_Direction_BS'!(X=@n12) -> (Pose) {blendSpace=/Hoagie/Animations/Vehicle/Hoagie_Engine_Boosting_Direction_BS})
    bAlphaBoolEnabled:bool=True
    Translation="0,0,0"
    Rotation=@n13
    Scale="1,1,1"
   ) -> (Pose) {
    alphaInputType=Bool
    rotationMode=BMM_Additive
    alphaBoolBlend {blendInTime=0.5, blendOption=Linear, blendOutTime=0.5}
    boneToModify {boneName=nozzle_1_l_end}
   }
   n15 = "Transform (Modify) Bone - Bone: nozzle_1_r_end"!(ComponentPose=@n14, bAlphaBoolEnabled:bool=True, Translation="0,0,0", Rotation=@n13, Scale="1,1,1") -> (Pose) {
    alphaInputType=Bool
    rotationMode=BMM_Additive
    alphaBoolBlend {blendInTime=0.5, blendOption=Linear, blendOutTime=0.5}
    boneToModify {boneName=nozzle_1_r_end}
   }
   n16 = Component To Local(ComponentPose=@n15)
   n17 = Use cached pose 'EngineBoostPose'()
   n18 = Get bShouldApplyBoostAdditive() -> (bShouldApplyBoostAdditive:bool)
   AnimGraphNode_StateMachine_0 = LeverMovement() {is=title AnimGraphNode_StateMachine}
   n19 = Layered blend per bone(
    BasePose=Layered blend per bone(BasePose=Local Space Ref Pose(), BlendPoses_0=@AnimGraphNode_StateMachine_0) {layerSetup [{branchFilters [{boneName=seat}, {boneName=pedals}]}]}
    BlendPoses_0=Blendspace Player 'Hoagie_Collective_BS'!(X=Get VerticalSpeed() -> (VerticalSpeed:float)) -> (Pose) {blendSpace=/Hoagie/Animations/Vehicle/Hoagie_Collective_BS}
   ) {layerSetup [{branchFilters [{boneName=collective}, {boneName=blade_l}, {boneName=blade_r}, {boneName=blade_c}]}]}
   n20 = Layered blend per bone(
    BasePose=@n19
    BlendPoses_0=Blend Poses by bool(
     BlendPose_0=Apply Additive!(Base=@n17, Additive=@n16, bAlphaBoolEnabled:bool=@n18) -> (Pose) {alphaInputType=Bool, alphaBoolBlend {blendInTime=0.2, blendOutTime=0.5}}
     BlendPose_1=Use cached pose 'EngineBoostPose'()
     BlendTime_0=0.25
     BlendTime_1=0.25
     bActiveValue=@n18
    ) {blendType=Linear, childUpateMode=ResetChildOnActivate}
   ) {layerSetup [{branchFilters [{boneName=nozzle_1_r}, {boneName=nozzle_1_l}]}]}
   n21 = Sequence Player 'Hoagie_Boost_OnCD_Vehicle'() {sequence=/Hoagie/Animations/Vehicle/Hoagie_Boost_OnCD_Vehicle}
   n22 = Get bHasDriver() -> (bHasDriver:bool)
   n23 = "Transform (Modify) Bone - Bone: Hoagie_tailrotor"!(
    ComponentPose=Local To Component(LocalPose=Blend Poses by bool(BlendPose_0=@n20, BlendPose_1=@n21, bActiveValue=@n22) {blendType=Linear})
    Alpha=1
    Rotation=Get TailRotorRotaton() -> (TailRotorRotaton:Rotator)
   ) -> (Pose) {rotationMode=BMM_Replace, rotationSpace=BCS_BoneSpace, alphaBoolBlend {blendOption=Linear}, boneToModify {boneName=Hoagie_tailrotor}}
   n24 = "Transform (Modify) Bone - Bone: Hoagie_mainrotor"!(ComponentPose=@n23, Alpha=1, Rotation=Get MainRotorRotation() -> (MainRotorRotation:Rotator)) -> (Pose) {rotationMode=BMM_Replace, rotationSpace=BCS_BoneSpace, alphaBoolBlend {blendOption=Linear}, boneToModify {boneName=Hoagie_mainrotor}}
   n25 = Get bShouldApplyBoostAdditive()
   n26 = Get BoosterFanRotation() -> (BoosterFanRotation:Rotator)
   n27 = "Transform (Modify) Bone - Bone: booster_fan_r"!(ComponentPose=@n24, bAlphaBoolEnabled:bool=@n25, Rotation=@n26) -> (Pose) {
    alphaInputType=Bool
    rotationMode=BMM_Replace
    rotationSpace=BCS_BoneSpace
    alphaBoolBlend {blendOption=Linear}
    boneToModify {boneName=booster_fan_r}
   }
   n28 = "Transform (Modify) Bone - Bone: booster_fan_l"!(ComponentPose=@n27, bAlphaBoolEnabled:bool=@n25, Rotation=@n26) -> (Pose) {
    alphaInputType=Bool
    rotationMode=BMM_Replace
    rotationSpace=BCS_BoneSpace
    alphaBoolBlend {blendOption=Linear}
    boneToModify {boneName=booster_fan_l}
   }
   n29 = Component To Local(ComponentPose=@n28)
   n30 = "Sequence Player (additive) 'Hoagie_ShutDown_Pose_Vehicle'"() {sequence=/Hoagie/Animations/Vehicle/Hoagie_ShutDown_Pose_Vehicle}
   n31 = Get bIsEngineShuttingOff() -> (bIsEngineShuttingOff:bool)
   n32 = "Constraint - blade_flex_pos_1_c (PPPP)"!(
    ComponentPose=Local To Component(LocalPose=Apply Additive(Base=@n29, Additive=@n30, bAlphaBoolEnabled=@n31) {alphaInputType=Bool, alphaBoolBlend {blendInTime=3, blendOutTime=1.5}})
    Alpha=1
    ConstraintWeights_0=0.4679
    ConstraintWeights_1:float=0.0037
    ConstraintWeights_2:float=0.4663
    ConstraintWeights_3:float=0.0621
   ) -> (Pose) {
    alphaBoolBlend {blendOption=Linear}
    boneToModify {boneName=blade_flex_pos_1_c}
    constraintSetup [
     {targetBone {boneName=blade_flex_1_c}}
     {targetBone {boneName=blade_flex_3_c}}
     {targetBone {boneName=blade_c}}
     {targetBone {boneName=blade_flex_2_c}}
    ]
    constraintWeights [1,1,1,1]
   }
   n33 = "Constraint - blade_flex_pos_2_c (PPPP)"!(
    ComponentPose=@n32
    Alpha=1
    ConstraintWeights_0=0.5245
    ConstraintWeights_1:float=0.0449
    ConstraintWeights_2:float=0.1053
    ConstraintWeights_3:float=0.3254
   ) -> (Pose) {
    alphaBoolBlend {blendOption=Linear}
    boneToModify {boneName=blade_flex_pos_2_c}
    constraintSetup [
     {targetBone {boneName=blade_flex_1_c}}
     {targetBone {boneName=blade_flex_3_c}}
     {targetBone {boneName=blade_c}}
     {targetBone {boneName=blade_flex_2_c}}
    ]
    constraintWeights [1,1,1,1]
   }
   n34 = "Constraint - blade_flex_pos_3_c (PPPP)"!(
    ComponentPose=@n33
    Alpha=1
    ConstraintWeights_0=0.3498
    ConstraintWeights_1:float=0.141
    ConstraintWeights_2:float=0.0242
    ConstraintWeights_3:float=0.485
   ) -> (Pose) {
    alphaBoolBlend {blendOption=Linear}
    boneToModify {boneName=blade_flex_pos_3_c}
    constraintSetup [
     {targetBone {boneName=blade_flex_1_c}}
     {targetBone {boneName=blade_flex_3_c}}
     {targetBone {boneName=blade_c}}
     {targetBone {boneName=blade_flex_2_c}}
    ]
    constraintWeights [1,1,1,1]
   }
   n35 = "Constraint - blade_flex_pos_4_c (PPPP)"!(
    ComponentPose=@n34
    Alpha=1
    ConstraintWeights_0=0.1403
    ConstraintWeights_1:float=0.3711
    ConstraintWeights_2:float=0.0033
    ConstraintWeights_3:float=0.4853
   ) -> (Pose) {
    alphaBoolBlend {blendOption=Linear}
    boneToModify {boneName=blade_flex_pos_4_c}
    constraintSetup [
     {targetBone {boneName=blade_flex_1_c}}
     {targetBone {boneName=blade_flex_3_c}}
     {targetBone {boneName=blade_c}}
     {targetBone {boneName=blade_flex_2_c}}
    ]
    constraintWeights [1,1,1,1]
   }
   n36 = "Constraint - blade_flex_pos_5_c (PPPP)"!(
    ComponentPose=@n35
    Alpha=1
    ConstraintWeights_0=0.0441
    ConstraintWeights_1:float=0.6239
    ConstraintWeights_2:float=0.0004
    ConstraintWeights_3:float=0.3316
   ) -> (Pose) {
    alphaBoolBlend {blendOption=Linear}
    boneToModify {boneName=blade_flex_pos_5_c}
    constraintSetup [
     {targetBone {boneName=blade_flex_1_c}}
     {targetBone {boneName=blade_flex_3_c}}
     {targetBone {boneName=blade_c}}
     {targetBone {boneName=blade_flex_2_c}}
    ]
    constraintWeights [1,1,1,1]
   }
   n37 = "Constraint - blade_flex_pos_6_c (PPPP)"!(
    ComponentPose=@n36
    Alpha=1
    ConstraintWeights_0=0.006
    ConstraintWeights_1:float=0.842
    ConstraintWeights_2:float=0
    ConstraintWeights_3:float=0.152
   ) -> (Pose) {
    alphaBoolBlend {blendOption=Linear}
    boneToModify {boneName=blade_flex_pos_6_c}
    constraintSetup [
     {targetBone {boneName=blade_flex_1_c}}
     {targetBone {boneName=blade_flex_3_c}}
     {targetBone {boneName=blade_c}}
     {targetBone {boneName=blade_flex_2_c}}
    ]
    constraintWeights [1,1,1,1]
   }
   n38 = "Constraint - blade_flex_pos_7_c (PPPP)"!(
    ComponentPose=@n37
    Alpha=1
    ConstraintWeights_0=0.0048
    ConstraintWeights_1:float=0.8524
    ConstraintWeights_2:float=0
    ConstraintWeights_3:float=0.1429
   ) -> (Pose) {
    alphaBoolBlend {blendOption=Linear}
    boneToModify {boneName=blade_flex_pos_7_c}
    constraintSetup [
     {targetBone {boneName=blade_flex_1_c}}
     {targetBone {boneName=blade_flex_3_c}}
     {targetBone {boneName=blade_c}}
     {targetBone {boneName=blade_flex_2_c}}
    ]
    constraintWeights [1,1,1,1]
   }
   n39 = Control Rig(Source=Component To Local(ComponentPose=@n38)) {
    alphaBoolBlend {blendOption=Linear}
    controlRigAssetReference {blueprintRigClass=&/SK_Hoagie_Protomesh_CtrlRig.SK_Hoagie_Protomesh_CtrlRig_C}
   }
   Output Pose(Result=@n39) {layerGroup=None}
   Save cached pose 'SteerDirectionPose'(Pose=Blendspace Player 'Hoagie_Lever_BS'(X=Get DriverRoll() -> (DriverRoll:float), Y=Get DriverYaw() -> (DriverYaw:float)) {blendSpace=/Hoagie/Animations/Vehicle/Hoagie_Lever_BS}) {cacheName=SteerDirectionPose}
   Save cached pose 'IdlePose'(Pose=Blend Poses by bool(
     BlendPose_0=Use cached pose 'SteerDirectionPose'()
     BlendPose_1=Sequence Player 'Hoagie_Idle_Pose_Vehicle'() {sequence=/Hoagie/Animations/Vehicle/Hoagie_Idle_Pose_Vehicle}
     BlendTime_0=0.35
     BlendTime_1=0.35
     bActiveValue=Get bIsMovingAnyDirection() -> (bIsMovingAnyDirection:bool)
    ) {blendType=Linear}) {cacheName=IdlePose}
   AnimGraphNode_StateMachine_1 = EngineBoost() {is=title AnimGraphNode_StateMachine}
   Save cached pose 'EngineBoostPose'(Pose=@AnimGraphNode_StateMachine_1) {cacheName=EngineBoostPose}
   Get LocalBoostDirection() -> (LocalBoostDirection:float)
  ]}
 LeverMovement {
  entry=Default
  path=AnimGraph.AnimGraphNode_StateMachine_0.LeverMovement
  ids=id
  nodes [
   AnimStateNode_0 = Default() {is=title AnimStateNode}
   AnimStateNode_1 = Drive_N_Start() {is=title AnimStateNode}
   AnimStateNode_2 = Drive_N_Loop() {is=title AnimStateNode}
   AnimStateNode_3 = Boost_Start() {is=title AnimStateNode}
   AnimStateNode_4 = Back_to_Idle_Transition() {is=title AnimStateNode}
   AnimStateNode_5 = Reverse_Start() {is=title AnimStateNode}
   AnimStateNode_6 = Reverse() {is=title AnimStateNode}
   AnimStateNode_7 = Reverse_End() {is=title AnimStateNode}
   AnimStateNode_8 = Boost_Loop() {is=title AnimStateNode}
   AnimStateTransitionNode_0 = AnimStateTransitionNode() {priorityOrder=2}
   AnimStateTransitionNode_1 = AnimStateTransitionNode() {bAutomaticRuleBasedOnSequencePlayerInState=true, crossfadeDuration=0.35}
   AnimStateTransitionNode_10 = AnimStateTransitionNode() {bAutomaticRuleBasedOnSequencePlayerInState=true, crossfadeDuration=0.3}
   AnimStateTransitionNode_11 = AnimStateTransitionNode()
   AnimStateTransitionNode_12 = AnimStateTransitionNode() {crossfadeDuration=0.3}
   AnimStateTransitionNode_13 = AnimStateTransitionNode()
   AnimStateTransitionNode_14 = AnimStateTransitionNode() {crossfadeDuration=0.1}
   AnimStateTransitionNode_15 = AnimStateTransitionNode() {crossfadeDuration=0.1}
   AnimStateTransitionNode_16 = AnimStateTransitionNode() {crossfadeDuration=0.1}
   AnimStateTransitionNode_17 = AnimStateTransitionNode() {crossfadeDuration=0.1}
   AnimStateTransitionNode_18 = AnimStateTransitionNode() {crossfadeDuration=0.1}
   AnimStateTransitionNode_19 = AnimStateTransitionNode() {crossfadeDuration=0.35}
   AnimStateTransitionNode_2 = AnimStateTransitionNode() {bAutomaticRuleBasedOnSequencePlayerInState=true, crossfadeDuration=0.15}
   AnimStateTransitionNode_20 = AnimStateTransitionNode() {crossfadeDuration=0.1}
   AnimStateTransitionNode_22 = AnimStateTransitionNode() {priorityOrder=2}
   AnimStateTransitionNode_3 = AnimStateTransitionNode() {priorityOrder=2}
   AnimStateTransitionNode_4 = AnimStateTransitionNode() {crossfadeDuration=0.3}
   AnimStateTransitionNode_5 = AnimStateTransitionNode() {bAutomaticRuleBasedOnSequencePlayerInState=true, priorityOrder=2}
   AnimStateTransitionNode_6 = AnimStateTransitionNode() {priorityOrder=2}
   AnimStateTransitionNode_7 = AnimStateTransitionNode() {bAutomaticRuleBasedOnSequencePlayerInState=true, crossfadeDuration=0.35}
   AnimStateTransitionNode_8 = AnimStateTransitionNode() {crossfadeDuration=0.35, priorityOrder=2}
   AnimStateTransitionNode_9 = AnimStateTransitionNode() {crossfadeDuration=0.1}
  ]
  links [
   {
    dst=Boost_Start
    rule=AnimGraph.AnimGraphNode_StateMachine_0.LeverMovement.AnimStateTransitionNode_18.Transition
    src=Back_to_Idle_Transition
    via=AnimStateTransitionNode_18
   }
   {
    dst=Default
    rule=AnimGraph.AnimGraphNode_StateMachine_0.LeverMovement.AnimStateTransitionNode_10.Transition
    src=Back_to_Idle_Transition
    via=AnimStateTransitionNode_10
   }
   {
    dst=Drive_N_Loop
    rule=AnimGraph.AnimGraphNode_StateMachine_0.LeverMovement.AnimStateTransitionNode_8.Transition
    src=Boost_Loop
    via=AnimStateTransitionNode_8
   }
   {
    dst=Reverse
    rule=AnimGraph.AnimGraphNode_StateMachine_0.LeverMovement.AnimStateTransitionNode_19.Transition
    src=Boost_Loop
    via=AnimStateTransitionNode_19
   }
   {
    dst=Boost_Loop
    rule=AnimGraph.AnimGraphNode_StateMachine_0.LeverMovement.AnimStateTransitionNode_2.Transition
    src=Boost_Start
    via=AnimStateTransitionNode_2
   }
   {
    dst=Boost_Start
    rule=AnimGraph.AnimGraphNode_StateMachine_0.LeverMovement.AnimStateTransitionNode_20.Transition
    src=Default
    via=AnimStateTransitionNode_20
   }
   {
    dst=Drive_N_Start
    rule=AnimGraph.AnimGraphNode_StateMachine_0.LeverMovement.AnimStateTransitionNode_0.Transition
    src=Default
    via=AnimStateTransitionNode_0
   }
   {
    dst=Reverse_Start
    rule=AnimGraph.AnimGraphNode_StateMachine_0.LeverMovement.AnimStateTransitionNode_12.Transition
    src=Default
    via=AnimStateTransitionNode_12
   }
   {
    dst=Back_to_Idle_Transition
    rule=AnimGraph.AnimGraphNode_StateMachine_0.LeverMovement.AnimStateTransitionNode_3.Transition
    src=Drive_N_Loop
    via=AnimStateTransitionNode_3
   }
   {
    dst=Boost_Start
    rule=AnimGraph.AnimGraphNode_StateMachine_0.LeverMovement.AnimStateTransitionNode_9.Transition
    src=Drive_N_Loop
    via=AnimStateTransitionNode_9
   }
   {
    dst=Reverse_Start
    rule=AnimGraph.AnimGraphNode_StateMachine_0.LeverMovement.AnimStateTransitionNode_4.Transition
    src=Drive_N_Loop
    via=AnimStateTransitionNode_4
   }
   {
    dst=Boost_Start
    rule=AnimGraph.AnimGraphNode_StateMachine_0.LeverMovement.AnimStateTransitionNode_14.Transition
    src=Drive_N_Start
    via=AnimStateTransitionNode_14
   }
   {
    dst=Default
    rule=AnimGraph.AnimGraphNode_StateMachine_0.LeverMovement.AnimStateTransitionNode_11.Transition
    src=Drive_N_Start
    via=AnimStateTransitionNode_11
   }
   {
    dst=Drive_N_Loop
    rule=AnimGraph.AnimGraphNode_StateMachine_0.LeverMovement.AnimStateTransitionNode_1.Transition
    src=Drive_N_Start
    via=AnimStateTransitionNode_1
   }
   {
    dst=Drive_N_Loop
    rule=AnimGraph.AnimGraphNode_StateMachine_0.LeverMovement.AnimStateTransitionNode_22.Transition
    src=Drive_N_Start
    via=AnimStateTransitionNode_22
   }
   {
    dst=Boost_Start
    rule=AnimGraph.AnimGraphNode_StateMachine_0.LeverMovement.AnimStateTransitionNode_16.Transition
    src=Reverse
    via=AnimStateTransitionNode_16
   }
   {
    dst=Drive_N_Start
    rule=AnimGraph.AnimGraphNode_StateMachine_0.LeverMovement.AnimStateTransitionNode_13.Transition
    src=Reverse
    via=AnimStateTransitionNode_13
   }
   {
    dst=Reverse_End
    rule=AnimGraph.AnimGraphNode_StateMachine_0.LeverMovement.AnimStateTransitionNode_6.Transition
    src=Reverse
    via=AnimStateTransitionNode_6
   }
   {
    dst=Boost_Start
    rule=AnimGraph.AnimGraphNode_StateMachine_0.LeverMovement.AnimStateTransitionNode_17.Transition
    src=Reverse_End
    via=AnimStateTransitionNode_17
   }
   {
    dst=Default
    rule=AnimGraph.AnimGraphNode_StateMachine_0.LeverMovement.AnimStateTransitionNode_7.Transition
    src=Reverse_End
    via=AnimStateTransitionNode_7
   }
   {
    dst=Boost_Start
    rule=AnimGraph.AnimGraphNode_StateMachine_0.LeverMovement.AnimStateTransitionNode_15.Transition
    src=Reverse_Start
    via=AnimStateTransitionNode_15
   }
   {
    dst=Reverse
    rule=AnimGraph.AnimGraphNode_StateMachine_0.LeverMovement.AnimStateTransitionNode_5.Transition
    src=Reverse_Start
    via=AnimStateTransitionNode_5
   }
  ]
 }
 Default {
  path=AnimGraph.AnimGraphNode_StateMachine_0.LeverMovement.AnimStateNode_0.Default
  nodes [
   Sequence Player 'Hoagie_Idle_Pose_Vehicle'() {sequence=&Hoagie_Idle_Pose_Vehicle}
   Output Animation Pose(Result=Use cached pose 'IdlePose'())
  ]
 }
 Drive_N_Start {
  path=AnimGraph.AnimGraphNode_StateMachine_0.LeverMovement.AnimStateNode_1.Drive_N_Start
  nodes [Output Animation Pose(Result=Apply Additive!(
     Base=Use cached pose 'IdlePose'()
     Additive="Sequence Player (additive) 'Hoagie_N_Start_Vehicle'"!(PlayRate:float=1) -> (Pose) {bLoopAnimation=false, sequence=/Hoagie/Animations/Vehicle/Hoagie_N_Start_Vehicle}
     Alpha:float
    ) -> (Pose) {alphaBoolBlend {blendOption=Linear}})]
 }
 Drive_N_Loop {
  path=AnimGraph.AnimGraphNode_StateMachine_0.LeverMovement.AnimStateNode_2.Drive_N_Loop
  nodes [Output Animation Pose(Result=Apply Additive(
     Base=Use cached pose 'IdlePose'()
     Additive="Sequence Player (additive) 'Hoagie_N_AdditivePose_Vehicle'"() {sequence=/Hoagie/Animations/Vehicle/Hoagie_N_AdditivePose_Vehicle}
    ) {alphaBoolBlend {blendOption=Linear}})]
 }
 Boost_Start {
  path=AnimGraph.AnimGraphNode_StateMachine_0.LeverMovement.AnimStateNode_3.Boost_Start
  nodes [
   n1 = Sequence Player 'Hoagie_Boost_Start_Vehicle'() {bLoopAnimation=false, groupName=LeverBoost, method=SyncGroup, sequence=/Hoagie/Animations/Vehicle/Hoagie_Boost_Start_Vehicle}
   n2 = Sequence Player 'Hoagie_Boost_Start_S_Vehicle'() {bLoopAnimation=false, groupName=LeverBoost, method=SyncGroup, sequence=/Hoagie/Animations/Vehicle/Hoagie_Boost_Start_S_Vehicle}
   n3 = Sequence Player 'Hoagie_Boost_Start_W_Vehicle'() {bLoopAnimation=false, groupName=LeverBoost, method=SyncGroup, sequence=/Hoagie/Animations/Vehicle/Hoagie_Boost_Start_W_Vehicle}
   n4 = Sequence Player 'Hoagie_Boost_Start_E_Vehicle'() {bLoopAnimation=false, groupName=LeverBoost, method=SyncGroup, sequence=/Hoagie/Animations/Vehicle/Hoagie_Boost_Start_E_Vehicle}
   n5 = "Blend Poses (EFortCardinalDirection)"!(
    BlendPose_0:Pose Link Structure=@n1
    BlendPose_1:Pose Link Structure=@n3
    BlendPose_2:Pose Link Structure=@n2
    BlendPose_3:Pose Link Structure=@n4
    BlendTime_0:float=0.1
    BlendTime_1:float=0.1
    BlendTime_2:float=0.1
    BlendTime_3:float=0.1
    ActiveEnumValue:Byte=Get BoostCardinalDirection() -> (BoostCardinalDirection:EFortCardinalDirection Enum=North)
   ) -> (Pose:Pose Link Structure) {
    is=title AnimGraphNode_BlendListByEnum
    blendType=Linear
    blendPose [{linkId=-1, sourceLinkId=-1}, {linkId=-1, sourceLinkId=-1}, {linkId=-1, sourceLinkId=-1}, {linkId=-1, sourceLinkId=-1}]
    blendTime [0.1,0.1,0.1,0.1]
   }
   Output Animation Pose(Result=@n5)
  ]
 }
 Back_to_Idle_Transition {
  path=AnimGraph.AnimGraphNode_StateMachine_0.LeverMovement.AnimStateNode_4.Back_to_Idle_Transition
  nodes [Output Animation Pose(Result=Sequence Player 'Hoagie_N_BackToIdle_Vehicle'() {bLoopAnimation=false, sequence=/Hoagie/Animations/Vehicle/Hoagie_N_BackToIdle_Vehicle})]
 }
 Reverse_Start {
  path=AnimGraph.AnimGraphNode_StateMachine_0.LeverMovement.AnimStateNode_5.Reverse_Start
  nodes [Output Animation Pose(Result=Sequence Player 'Hoagie_Reverse_Start_Vehicle'() {bLoopAnimation=false, sequence=/Hoagie/Animations/Vehicle/Hoagie_Reverse_Start_Vehicle})]
 }
 Reverse {
  path=AnimGraph.AnimGraphNode_StateMachine_0.LeverMovement.AnimStateNode_6.Reverse
  nodes [Output Animation Pose(Result=Sequence Player 'Hoagie_Reverse_Pose_Vehicle'() {sequence=/Hoagie/Animations/Vehicle/Hoagie_Reverse_Pose_Vehicle})]
 }
 Reverse_End {
  path=AnimGraph.AnimGraphNode_StateMachine_0.LeverMovement.AnimStateNode_7.Reverse_End
  nodes [Output Animation Pose(Result=Sequence Player 'Hoagie_Reverse_End_Vehicle'() {bLoopAnimation=false, sequence=/Hoagie/Animations/Vehicle/Hoagie_Reverse_End_Vehicle})]
 }
 Boost_Loop {
  path=AnimGraph.AnimGraphNode_StateMachine_0.LeverMovement.AnimStateNode_8.Boost_Loop
  nodes [
   n1 = "Blend Poses (EFortCardinalDirection)"(
    BlendPose_0=Sequence Player 'Hoagie_Boost_Loop_Vehicle'() {sequence=/Hoagie/Animations/Vehicle/Hoagie_Boost_Loop_Vehicle}
    BlendPose_1=Sequence Player 'Hoagie_Boost_Loop_W_Vehicle'() {sequence=/Hoagie/Animations/Vehicle/Hoagie_Boost_Loop_W_Vehicle}
    BlendPose_2=Sequence Player 'Hoagie_Boost_Loop_S_Vehicle'() {sequence=/Hoagie/Animations/Vehicle/Hoagie_Boost_Loop_S_Vehicle}
    BlendPose_3=Sequence Player 'Hoagie_Boost_Loop_E_Vehicle'() {sequence=/Hoagie/Animations/Vehicle/Hoagie_Boost_Loop_E_Vehicle}
    BlendTime_0=0.1
    BlendTime_1=0.1
    BlendTime_2=0.1
    BlendTime_3=0.1
    ActiveEnumValue=Get BoostCardinalDirection() -> (BoostCardinalDirection=North)
   ) {
    blendType=Linear
    blendPose [{linkId=-1, sourceLinkId=-1}, {linkId=-1, sourceLinkId=-1}, {linkId=-1, sourceLinkId=-1}, {linkId=-1, sourceLinkId=-1}]
    blendTime [0.1,0.1,0.1,0.1]
   }
   Output Animation Pose(Result=@n1)
  ]
 }
 Transition {
  path=AnimGraph.AnimGraphNode_StateMachine_0.LeverMovement.AnimStateTransitionNode_0.Transition
  nodes [Result(bCanEnterTransition=Get bIsMovingForward() -> (bIsMovingForward:bool))]
 }
 Transition {path=AnimGraph.AnimGraphNode_StateMachine_0.LeverMovement.AnimStateTransitionNode_1.Transition, nodes [Result()]}
 Transition {path=AnimGraph.AnimGraphNode_StateMachine_0.LeverMovement.AnimStateTransitionNode_10.Transition, nodes [Result()]}
 Transition {
  path=AnimGraph.AnimGraphNode_StateMachine_0.LeverMovement.AnimStateTransitionNode_11.Transition
  nodes [Result(bCanEnterTransition=Get bIsReversing() -> (bIsReversing:bool))]
 }
 Transition {
  path=AnimGraph.AnimGraphNode_StateMachine_0.LeverMovement.AnimStateTransitionNode_12.Transition
  nodes [Result(bCanEnterTransition=Get bTransition_Default_ReverseStart() -> (bTransition_Default_ReverseStart:bool))]
 }
 Transition {
  path=AnimGraph.AnimGraphNode_StateMachine_0.LeverMovement.AnimStateTransitionNode_13.Transition
  nodes [Result(bCanEnterTransition=Get bIsMovingForward())]
 }
 Transition {
  path=AnimGraph.AnimGraphNode_StateMachine_0.LeverMovement.AnimStateTransitionNode_14.Transition
  nodes [Result(bCanEnterTransition=Get bIsBoosting() -> (bIsBoosting:bool))]
 }
 Transition {
  path=AnimGraph.AnimGraphNode_StateMachine_0.LeverMovement.AnimStateTransitionNode_15.Transition
  nodes [Result(bCanEnterTransition=Get bIsBoosting())]
 }
 Transition {
  path=AnimGraph.AnimGraphNode_StateMachine_0.LeverMovement.AnimStateTransitionNode_16.Transition
  nodes [Result(bCanEnterTransition=Get bIsBoosting())]
 }
 Transition {
  path=AnimGraph.AnimGraphNode_StateMachine_0.LeverMovement.AnimStateTransitionNode_17.Transition
  nodes [Result(bCanEnterTransition=Get bIsBoosting())]
 }
 Transition {
  path=AnimGraph.AnimGraphNode_StateMachine_0.LeverMovement.AnimStateTransitionNode_18.Transition
  nodes [Result(bCanEnterTransition=Get bIsBoosting())]
 }
 Transition {
  path=AnimGraph.AnimGraphNode_StateMachine_0.LeverMovement.AnimStateTransitionNode_19.Transition
  nodes [Result(bCanEnterTransition=Get bTransition_BoostLoop_Reverse() -> (bTransition_BoostLoop_Reverse:bool))]
 }
 Transition {path=AnimGraph.AnimGraphNode_StateMachine_0.LeverMovement.AnimStateTransitionNode_2.Transition, nodes [Result()]}
 Transition {
  path=AnimGraph.AnimGraphNode_StateMachine_0.LeverMovement.AnimStateTransitionNode_20.Transition
  nodes [Result(bCanEnterTransition=Get bIsBoosting())]
 }
 Transition {
  path=AnimGraph.AnimGraphNode_StateMachine_0.LeverMovement.AnimStateTransitionNode_22.Transition
  nodes [Result(bCanEnterTransition=NOT Boolean(A=Get bIsMovingAnyDirection()))]
 }
 Transition {
  path=AnimGraph.AnimGraphNode_StateMachine_0.LeverMovement.AnimStateTransitionNode_3.Transition
  nodes [Result(bCanEnterTransition=Get bShouldGoBackToIdle() -> (bShouldGoBackToIdle:bool))]
 }
 Transition {
  path=AnimGraph.AnimGraphNode_StateMachine_0.LeverMovement.AnimStateTransitionNode_4.Transition
  nodes [Result(bCanEnterTransition=Get bShouldReverse() -> (bShouldReverse:bool))]
 }
 Transition {path=AnimGraph.AnimGraphNode_StateMachine_0.LeverMovement.AnimStateTransitionNode_5.Transition, nodes [Result()]}
 Transition {
  path=AnimGraph.AnimGraphNode_StateMachine_0.LeverMovement.AnimStateTransitionNode_6.Transition
  nodes [Result(bCanEnterTransition=NOT Boolean(A=Get bIsReversing()))]
 }
 Transition {path=AnimGraph.AnimGraphNode_StateMachine_0.LeverMovement.AnimStateTransitionNode_7.Transition, nodes [Result()]}
 Transition {
  path=AnimGraph.AnimGraphNode_StateMachine_0.LeverMovement.AnimStateTransitionNode_8.Transition
  nodes [Result(bCanEnterTransition=NOT Boolean(A=Get bIsBoosting()))]
 }
 Transition {
  path=AnimGraph.AnimGraphNode_StateMachine_0.LeverMovement.AnimStateTransitionNode_9.Transition
  nodes [Result(bCanEnterTransition=Get bIsBoosting())]
 }
 EngineBoost {
  entry=Default
  path=AnimGraph.AnimGraphNode_StateMachine_1.EngineBoost
  ids=id
  nodes [
   AnimStateNode_0 = Default()
   AnimStateNode_1 = Start() {is=title AnimStateNode}
   AnimStateNode_2 = Loop() {is=title AnimStateNode}
   AnimStateNode_3 = End() {is=title AnimStateNode}
   AnimStateNode_4 = Release() {is=title AnimStateNode}
   AnimStateTransitionNode_1 = AnimStateTransitionNode() {crossfadeDuration=0}
   AnimStateTransitionNode_3 = AnimStateTransitionNode() {bAutomaticRuleBasedOnSequencePlayerInState=true, crossfadeDuration=0}
   AnimStateTransitionNode_5 = AnimStateTransitionNode() {bAutomaticRuleBasedOnSequencePlayerInState=true}
   AnimStateTransitionNode_6 = AnimStateTransitionNode() {crossfadeDuration=0.1}
   AnimStateTransitionNode_7 = AnimStateTransitionNode()
   AnimStateTransitionNode_9 = AnimStateTransitionNode() {crossfadeDuration=0.1, priorityOrder=2}
  ]
  links [
   {
    dst=Start
    rule=AnimGraph.AnimGraphNode_StateMachine_1.EngineBoost.AnimStateTransitionNode_1.Transition
    src=Default
    via=AnimStateTransitionNode_1
   }
   {
    dst=Default
    rule=AnimGraph.AnimGraphNode_StateMachine_1.EngineBoost.AnimStateTransitionNode_5.Transition
    src=End
    via=AnimStateTransitionNode_5
   }
   {
    dst=Release
    rule=AnimGraph.AnimGraphNode_StateMachine_1.EngineBoost.AnimStateTransitionNode_6.Transition
    src=Loop
    via=AnimStateTransitionNode_6
   }
   {
    dst=End
    rule=AnimGraph.AnimGraphNode_StateMachine_1.EngineBoost.AnimStateTransitionNode_7.Transition
    src=Release
    via=AnimStateTransitionNode_7
   }
   {
    dst=Loop
    rule=AnimGraph.AnimGraphNode_StateMachine_1.EngineBoost.AnimStateTransitionNode_3.Transition
    src=Start
    via=AnimStateTransitionNode_3
   }
   {
    dst=Release
    rule=AnimGraph.AnimGraphNode_StateMachine_1.EngineBoost.AnimStateTransitionNode_9.Transition
    src=Start
    via=AnimStateTransitionNode_9
   }
  ]
 }
 Default {
  path=AnimGraph.AnimGraphNode_StateMachine_1.EngineBoost.AnimStateNode_0.Default
  nodes [Output Animation Pose(Result=Sequence Player 'Hoagie_Boost_OnCD_Vehicle'() {sequence=&Hoagie_Boost_OnCD_Vehicle})]
 }
 Start {
  path=AnimGraph.AnimGraphNode_StateMachine_1.EngineBoost.AnimStateNode_1.Start
  nodes [Output Animation Pose(Result=Sequence Player 'Hoagie_Boost_Ready_Vehicle'() {bLoopAnimation=false, sequence=/Hoagie/Animations/Vehicle/Hoagie_Boost_Ready_Vehicle})]
 }
 Loop {
  path=AnimGraph.AnimGraphNode_StateMachine_1.EngineBoost.AnimStateNode_2.Loop
  nodes [Output Animation Pose(Result=Apply Additive(
     Base=Sequence Player 'Hoagie_Boost_Ready_Loop_Vehicle'() {sequence=/Hoagie/Animations/Vehicle/Hoagie_Boost_Ready_Loop_Vehicle}
     Additive="Sequence Player (additive) 'Hoagie_Boost_Ready_Loop_Additive_Vehicle'"() {sequence=/Hoagie/Animations/Vehicle/Hoagie_Boost_Ready_Loop_Additive_Vehicle}
    ) {alphaBoolBlend {blendOption=Linear}})]
 }
 End {
  path=AnimGraph.AnimGraphNode_StateMachine_1.EngineBoost.AnimStateNode_3.End
  nodes [Output Animation Pose(Result=Sequence Player 'Hoagie_Boost_End_Vehicle'() {bLoopAnimation=false, sequence=/Hoagie/Animations/Vehicle/Hoagie_Boost_End_Vehicle})]
 }
 Release {
  path=AnimGraph.AnimGraphNode_StateMachine_1.EngineBoost.AnimStateNode_4.Release
  nodes [Output Animation Pose(Result=Sequence Player 'Hoagie_Boost_Release_Vehicle'() {sequence=/Hoagie/Animations/Vehicle/Hoagie_Boost_Release_Vehicle})]
 }
 Transition {
  path=AnimGraph.AnimGraphNode_StateMachine_1.EngineBoost.AnimStateTransitionNode_1.Transition
  nodes [Result(bCanEnterTransition=Get bIsBoostReady() -> (bIsBoostReady:bool))]
 }
 Transition {path=AnimGraph.AnimGraphNode_StateMachine_1.EngineBoost.AnimStateTransitionNode_3.Transition, nodes [Result()]}
 Transition {path=AnimGraph.AnimGraphNode_StateMachine_1.EngineBoost.AnimStateTransitionNode_5.Transition, nodes [Result()]}
 Transition {
  path=AnimGraph.AnimGraphNode_StateMachine_1.EngineBoost.AnimStateTransitionNode_6.Transition
  nodes [Result(bCanEnterTransition=Get bIsBoosting())]
 }
 Transition {
  path=AnimGraph.AnimGraphNode_StateMachine_1.EngineBoost.AnimStateTransitionNode_7.Transition
  nodes [Result(bCanEnterTransition=Get bIsBoostOnCD() -> (bIsBoostOnCD:bool))]
 }
 Transition {
  path=AnimGraph.AnimGraphNode_StateMachine_1.EngineBoost.AnimStateTransitionNode_9.Transition
  nodes [Result(bCanEnterTransition=Get bIsBoosting())]
 }
 EventGraph {links [], nodes []}
]
variables [
 BladeScale {type=Vector, default=<0.001 0.001 0.001>}
 IsStarted {type=boolean}
]
