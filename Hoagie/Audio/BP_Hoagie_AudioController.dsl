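; Audio controller for the Hoagie rotorcraft vehicle, voicing engine, rotor wash, boost, horn, impacts, water entry and destruction.
; Caches vehicle and components for native code, then feeds vehicle state into layered engine and detached rotor wash parameters, some only in high-quality mode.
; One-shot damage thresholds swap sounds; critical damage winds the horn down via a pitch/volume timeline.
; A retriggerable delay brackets continuous rotor hits with start and end sounds.
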(event Custom|On%20Notify%20Rotor%20Outer%20Impact
  (bind _rotordamage (Variables|Default|GetRotorDamage))
  (bind _rotordamage_2 (Variables|Default|GetRotorDamage))
  ;[1 1 1] Rotor impacting environment
  (if (not (Variables|Vehicle|GetbRotorHitNative))
    (Audio|Components|Audio|SetSound _rotordamage_2 (Variables|Default|GetRotorHitStart))
    (Audio|Components|Audio|Play _rotordamage_2 0.0))
  (Variables|Vehicle|SetbRotorHitNative true)
  (Utilities|FlowControl|RetriggerableDelay 0.5)
  (Variables|Vehicle|SetbRotorHitNative false)
  (Audio|Components|Audio|SetSound _rotordamage (Variables|Default|GetRotorHitEnd))
  (Audio|Components|Audio|Play _rotordamage 0.0))

(event Custom|Update%20Damage%20State
  (params (Damage :type Float%20%28double-precision%29))
  (bind _returnvalue (- 1.0 Damage))
  (bind _damagestate (Variables|Default|GetDamageState))
  (bind _damagethresholds (Variables|Default|GetDamageThresholds))
  ;[1 1 1] Damage state events
  (Utilities|FlowControl|Sequence
    (:then_0
      (if (<= _returnvalue (Utilities|Array|Get(aref) _damagethresholds))
        (Utilities|FlowControl|DoOnce false
          (:Completed
            (Audio|Components|Audio|SetSound _damagestate (Variables|Default|GetDamageState1))
            (Audio|Components|Audio|Play _damagestate 0.0)))))
    (:then_1
      (if (<= _returnvalue (Utilities|Array|Get(aref) _damagethresholds 1))
        (Utilities|FlowControl|DoOnce false
          (:Completed
            (Audio|Components|Audio|SetSound _damagestate (Variables|Default|GetDamageState2))
            (Audio|Components|Audio|Play _damagestate 0.0)))))
    (:then_2
      (if (and (<= _returnvalue (Utilities|Array|Get(aref) _damagethresholds 2)) (not (Variables|Default|GetbIsDestroyed)))
        (Utilities|FlowControl|DoOnce false
          (:Completed
            (Audio|Components|Audio|SetSound _damagestate (Variables|Default|GetDamageState3))
            (Audio|Components|Audio|Play _damagestate 0.0)
            (Variables|Default|SetCriticalDamage true)
            (CallFunction|StopHorn false))))))
  (Class|FortLayeredAudioComponent|SetFloatParameterExt (Variables|Default|GetEngine) "EngineDuck" 1.0))

(event Custom|On%20Notify%20Boost%20Start
  ;[1 1 1] boost
  (Audio|Components|Audio|Play (Variables|Default|GetBoost) 0.0)
  (Audio|Components|Audio|FadeOut (Variables|Default|GetBoostReady) 0.5 0.0 "Linear" :!class /Script/Engine.AudioComponent))

(event Custom|On%20Notify%20Boost%20End)

(event Custom|On%20Notify%20Boost%20Ready
  ;[1 1 1] boost
  (Audio|Components|Audio|Play (Variables|Default|GetBoostReady) 0.0)
  (Audio|Components|Audio|FadeOut (Variables|Default|GetBoost) 0.5 0.0 "Linear" :!class /Script/Engine.AudioComponent))

(event Custom|On%20Notify%20Destroyed
  ;[1 1 1] Destroyed
  (Audio|PlaySoundatLocation (Variables|Assets|GetDestroySound) (Transformation|GetActorLocation))
  (Audio|Components|Audio|FadeOut (Variables|Default|GetDamageState) 0.2 0.0 "Linear" :!class /Script/Engine.AudioComponent)
  (Components|Timeline|Stop (Variables|BP_Hoagie_AudioController|GetHorn%20Pitch%20Down))
  (Audio|Components|Audio|Stop (Variables|Default|GetHorn))
  (Variables|Default|SetbIsDestroyed true))

(event Custom|On%20Notify%20Stop%20Tick
  ;[1 1 1] On Stop Tick
  (Audio|Components|Audio|FadeOut (Variables|Default|GetEngine) 1.0 0.0 "Linear" :!class /Script/Engine.AudioComponent))

(event Custom|On%20Notify%20Glass%20Destroyed
  ;[1 1 1] Glass destroyed events
  (Audio|PlaySoundatLocation "/Hoagie/Audio/Assets/Hoagie_DamageState_Glass_Cue.Hoagie_DamageState_Glass_Cue" (Transformation|GetActorLocation)))

(event Custom|On%20Notify%20Pilot%20Enter
  ;[1 1 1] pilot enter exit
  (Audio|Components|Audio|FadeIn (Variables|Default|GetRotorWash) 1.0 1.0 0.0 "Linear" :!class /Script/Engine.AudioComponent))

(event Custom|On%20Notify%20Pilot%20Exit
  ;[1 1 1] pilot enter exit
  (CallFunction|StopHorn true))

(event Custom|On%20Notify%20Collision%20Hit
  (params (HitLocation :type Vector%20%28by%20ref%29) (Normal%20Impulse :type Vector))
  ;[1 1 1] Collision audio
  (Audio|OnCollision (Variables|Default|GetFortCollisionAudioImpacts) HitLocation Normal%20Impulse))

(event Custom|Stop%20Horn
  (params (Manually%20Stopped :type Boolean))
  ;[1 1 1] Start / Stop Horn
  (if Manually%20Stopped
    (Audio|Components|Audio|FadeOut (Variables|Default|GetHorn) 1.0 0.0 "Linear" :!class /Script/Engine.AudioComponent)
    (else
      (bind (_direction _horn_pitch _horn_volume) (Timeline|Horn%20Pitch%20Down PlayFromStart)
        (:Update
          (Audio|Components|Audio|SetPitchMultiplier (Variables|Default|GetHorn) _horn_pitch)
          (Audio|Components|Audio|SetVolumeMultiplier (Variables|Default|GetHorn) _horn_volume :!class /Script/Engine.AudioComponent))
        (:Finished
          (switch Utilities|FlowControl|Switch|SwitchonETimelineDirection _direction
            (:Forward
              (Audio|Components|Audio|Stop (Variables|Default|GetHorn)))))))))

(event Custom|Update%20State%20Parameters
  ;[1 1 1] Update state-based parameters
  (Class|FortLayeredAudioComponent|SetFloatParameterExt (Variables|Default|GetEngine) "HasDriver" (Math|Conversions|ToFloat(Boolean) (Vehicle|HasDriver (Variables|Default|GetHoagieVehicle))))
  (Class|FortLayeredAudioComponent|SetFloatParameterExt (Variables|Default|GetEngine) "EngineDuck" (Math|Conversions|ToFloat(Boolean) (Class|FortVehicleInterface|IsBoosting (Variables|Default|GetHoagieVehicle))))
  (Class|FortLayeredAudioComponent|SetFloatParameterExt (Variables|Default|GetEngine) "IsLocalPlayerInVehicle" (Math|Conversions|ToFloat(Boolean) (Variables|Default|GetIsLocalPlayerInVehicle)))
  (Utilities|FlowControl|Sequence
    (:then_0
      (Utilities|HighQualityMode
        (:HighQualityMode
          (Class|FortLayeredAudioComponent|SetFloatParameterExt (Variables|Default|GetEngine) "IsSkidding" (Math|Conversions|ToFloat(Boolean) (Vehicle|IsScrapingBottom (Variables|Default|GetHoagieVehicle)))))))
    (:then_1
      (if (Variables|Default|GetCriticalDamage)
        (Class|FortLayeredAudioComponent|SetFloatParameterExt (Variables|Default|GetEngine) "CriticalDamage" 1.0)))))

(event Custom|Start%20Horn
  ;[1 1 1] Start / Stop Horn
  (if (not (Variables|Vehicle|GetbCriticalDamageNative))
    (Audio|Components|Audio|Play (Variables|Default|GetHorn) 0.0)
    (Utilities|FlowControl|Sequence
      (:then_0
        (Timeline|Horn%20Pitch%20Down SetNewTime))
      (:then_1
        (Timeline|Horn%20Pitch%20Down Stop)))))

(event Custom|On%20Horn%20Button%20Pressed
  ;[1 1 1] horn
  (if (Audio|Components|Audio|IsPlaying (Variables|Default|GetHorn))
    (CallFunction|StopHorn true)
    (else
      (CallFunction|StartHorn))))

(event Custom|Cache%20References
  (params (Vehicle :type Hoagie%20Vehicle%20Object%20Reference))
  ;[1 1 1] Caching for native
  (Vehicle|CacheHoagieVehicle Vehicle)
  (Variables|Default|SetHoagieVehicle Vehicle)
  (Vehicle|CacheAudioComponents (Variables|Default|GetEngine) (Variables|Default|GetRotorWash))
  (Transformation|DetachFromComponent (Variables|Default|GetRotorWash) "KeepRelative" "KeepRelative" "KeepRelative" true))

(event Custom|On%20Update%20Audio
  (params (HoagieVehicle :type Hoagie%20Vehicle%20Object%20Reference)))

(event Custom|Update%20Rotor%20Wash
  (bind _hoagievehicle (Variables|Default|GetHoagieVehicle))
  (bind _active (Class|HoagieVehicle|GetIsRotorWashActive _hoagievehicle :!class /Hoagie/HoagieVehicle.HoagieVehicle_C))
  (bind (_relativelocation_x _relativelocation_y _relativelocation_z) (Math|Vector|BreakVector (Variables|/Script/Engine.SceneComponent|GetRelativeLocation_X (Variables|/Hoagie/HoagieVehicle.HoagieVehicle_C|GetRotorWashFX (Variables|Default|GetHoagieVehicle)))))
  ;[1 1 1] Rotor Wash Update
  (Utilities|HighQualityMode
    (:HighQualityMode
      (Class|FortLayeredAudioComponent|SetFloatParameterExt (Variables|Default|GetRotorWash) "IsLocalPlayerInVehicle" (Math|Conversions|ToFloat(Boolean) (Variables|Default|GetIsLocalPlayerInVehicle)))
      (Class|FortLayeredAudioComponent|SetFloatParameterExt (Variables|Default|GetRotorWash) "Amount" (* (Math|Conversions|ToFloat(Boolean) _active) (Vehicle|GetRotorSpeedPercent (Variables|Default|GetHoagieVehicle))))
      (if _active
        (Transformation|SetRelativeLocation (Variables|Default|GetRotorWash) (Math|Vector|MakeVector 0.0 0.0 (select _active _relativelocation_z 0.0)) false false)))))

(event Custom|OnEnteredWater
  (params (Location :type Vector))
  ;[1 1 1] Water Splash Audio
  (Audio|PlaySoundatLocation (Variables|Default|GetWaterEnterSoundCue) Location))

(fn ConstructionScript)
