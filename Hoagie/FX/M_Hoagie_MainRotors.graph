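# Translucent, two-sided lit material for the Hoagie main rotors, drawing spinning blades from a propeller texture.
# Texture alpha is scaled by a clamped Alpha parameter and near-camera fade, alongside squared texture color.
# A sun/moon reflection dot product, raised to a power and tinted by time of day, forms a specular highlight.
# A panned gradient on radial UVs, masked by the same fade, drives refraction, replaced by a constant on mobile.
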Material /Hoagie/FX/M_Hoagie_MainRotors
graph {
 properties {
  bCastRayTracedShadows=false
  bEnableResponsiveAA=true
  bUsedWithMeshDeformer=true
  bUsedWithSkeletalMesh=true
  blendMode=BLEND_Translucent
  ditherOpacityMask=true
  pixelDepthOffsetMode=PDOM_Legacy
  refractionMethod=RM_IndexOfRefraction
  translucencyLightingMode=TLM_Surface
  twoSided=true
  usedShadingModels=Default Lit
 }
 nodes [
  Constant3Vector() {constant=<rgba 0 0 1 0>}
  n1 = TextureSampleParameter2D() {group=Texture, parameterName=PropellerTexture, samplerSource=SSM_Clamp_WorldGroupSettings, texture=&/T_FX_BladeMesh}
  LinearInterpolate(Alpha=OneMinus(None=@n1.R)) {constA=0.3, constB=0.1}
  n2 = Multiply(
   A=Clamp(None=ScalarParameter() {group=Alpha, parameterName=Alpha, sortPriority=1}) {maxDefault=0.9}
   B=MaterialFunctionCall(Distance=ScalarParameter() {defaultValue=250, parameterName=Near Camera Fade Distance}) {materialFunction=&Near_Camera_Fade}.Result
  )
  Multiply(A=@n1.A, B=@n2)
  n3 = MaterialFunctionCall() {materialFunction=&Color_Time_of_Day}
  n4 = CollectionParameter() {
   collection=&FortniteMaterialParameters
   parameterName=SunAndMoonModelDirectionalVector
  }
  n5 = DotProduct(A=ComponentMask(None=@n4) {b=true, g=true, r=true}, B=ReflectionVectorWS() {bNormalizeCustomWorldNormal=true})
  n6 = ScalarParameter() {defaultValue=44, group=Specular, parameterName=Spec Highlight Power}
  Multiply(A=Multiply(A=Add(A=@n3.Result) {constB=0.25}) {constB=0.5}, B=Clamp(None=Power(Base=@n5, Exp=@n6) {constExponent=20}))
  Multiply(A=@n1.RGB, B=@n1.RGB)
  n7 = TextureCoordinate()
  n8 = ComponentMask(None=VectorParameter() {parameterName=RefractionUVScale, defaultValue=<rgba 5 5 0 0>}.RGB) {g=true, r=true}
  n9 = MaterialFunctionCall(Vector or UVs=Subtract(A=Multiply(A=@n7, B=@n8), B=Multiply(A=@n8) {constB=0.5})) {materialFunction=&VectorToRadialValue}
  n10 = Multiply(
   A=@n2
   B=TextureSample(UVs=Panner(Coordinate=@n9.Radial Coordinates) {speedY=-2}) {texture=/Game/Global/Textures/Default/Masks_and_Gradients/gradient_stasis}.R
  )
  ShadingPathSwitch(
   Default=LinearInterpolate(B=ScalarParameter() {defaultValue=1, parameterName=Refraction Amount}, Alpha=Saturate(None=@n10)) {constA=1, constB=1.2}
   Mobile=Constant()
  )
 ]
}
