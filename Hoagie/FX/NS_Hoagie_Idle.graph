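# Niagara idle effect for Hoagie that shows boost readiness at two skeletal mesh sockets and adds speed lines as speed rises.
# Boost_Ready pins glowing sprites and lights to the sockets, while FlameOuterStart pulses short-lived flame meshes there, both faded by User.BoostReady.
# User.Speed curves drive speed-line spawn rate, size and alpha, with spawning starting above a threshold and distance culling applied.
# User.isCritical is declared but unreferenced in the shown modules.
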NiagaraSystem /Hoagie/FX/NS_Hoagie_Idle
dependencies {
 usedDataInterfaces [
  &NiagaraDataInterfaceVectorCurve
  &NiagaraDataInterfaceCurve
  &NiagaraDataInterfaceColorCurve
  &NiagaraDataInterfaceVector2DCurve
 ]
 usedDynamicInputs []
 usedModules [&SystemState]
}
emitters {
 Boost_Ready {
  emitterUpdateScript [
   &EmitterState(Life Cycle Mode=::NewEnumerator1, Loop Behavior=::NewEnumerator1, Loop Duration Mode=::NewEnumerator1, UseLoopDelay=false)
   &SpawnBurst_Instantaneous(Spawn Count=2, Use Spawn Probability=false)
  ]
  particleSpawnScript [
   /Niagara/Modules/Spawn/Initialization/InitializeParticle(
    Color=<rgba 1 0.1833 0 10>
    Lifetime=0
    Module.Write Color=true
    Module.Write Lifetime=true
    Module.Write Mass=false
    Module.Write Material Random=true
    Module.Write Position=false
    Module.Write Scale=false
    Module.Write SpriteRotation=false
    Module.Write SpriteSize=true
    Position=&SimulationPosition()
    Sprite Size=<20 20>
   )
   /Game/Effects/Niagara/Modules/Groups/FN_SetSpawnGroup_v1/FN_SetSpawnGroup(Emitter Bone Count=2)
  ]
  particleUpdateScript [
   &ParticleState(Kill Particles When Lifetime Has Elapsed=false, Let Infinitely Lived Particles Die When Emitter Deactivates=true)
   &SolveForcesAndVelocity(Clamp Velocity=false, Limit Acceleration=false)
   &SampleSkeletalMeshSkeleton(
    Mesh:NiagaraDataInterfaceSkeletalMesh=$User.SkeletalMesh
    Spawn Mode:ENiagaraSkeletalSpawning=::NewEnumerator1
    Specific Socket Index:NiagaraInt32=$Particles.SpawnGroup
   )
   /Niagara/Modules/Spawn/Location/SkeletalMeshSkeletonLocation(
    Module.Inherit Velocity=false
    Module.Use Sampled Bone as Ribbon ID=true
    Sampled Position Offset:Vector3f=&RotateVectorByQuaternion(Vector To Rotate=<-10 0 0>, Quaternion=$Particles.SampleSkeletalMeshSkeleton.SampledRotation)
   )
   /NiagaraFN/Sprite/FN_SpriteAlignmentAndFacing(
    SpriteAlignment:Vector3f=<0 0 1>
    SpriteFacing:Vector3f=&RotateVectorByQuaternion(Vector To Rotate=<1 0 0>, Quaternion=$Particles.SampleSkeletalMeshSkeleton.SampledRotation)
   ) {module=SpriteAlignmentAndFacing}
   &ScaleColor(Scale Alpha=$User.BoostReady, Scale RGB=<10 10 10>, ScaleA=true, ScaleRGB=true)
   /Game/Effects/Niagara/Modules/Light/FN_Light_v5/FN_Light(
    Light Exponent=2
    Light Radius:NiagaraFloat=&Multiply_Float(A=7, B=$User.BoostReady)
    LightEnabled=true
    Scale Color=1
   )
  ]
  rendererClasses [&NiagaraSpriteRendererProperties, &NiagaraLightRendererProperties]
  renderers [
   {rendererClass=&NiagaraSpriteRendererProperties}
   {rendererClass=&NiagaraLightRendererProperties, rendererIndex=1}
  ]
 }
 FN_Sprite_SpeedLines001 {
  emitterUpdateScript [
   &SpawnBurst_Instantaneous(Spawn Count=5, Use Spawn Probability=false) {enabled=false}
   &EmitterState()
   &SpawnRate(
    SpawnRate=&Lerp_Float(
     A=0
     B=&ClampFloat(Float=1, Min=1)
     Alpha=&Multiply_Float(
      A=&FloatFromCurve(
       FloatCurve=DataInterface() {propertyValues {Curve {keys [
           {arriveTangent=0.4214, interpMode=RCIM_Cubic, leaveTangent=0.4225, tangentMode=RCTM_User, time=25, value=0}
           {arriveTangent=17.7333, interpMode=RCIM_Cubic, leaveTangent=17.7333, tangentMode=RCTM_User, time=31.9282, value=72.4651}
           {arriveTangent=7.1685, interpMode=RCIM_Cubic, leaveTangent=7.1685, tangentMode=RCTM_User, time=39.6367, value=353.7616}
           {arriveTangent=16.7131, interpMode=RCIM_Cubic, leaveTangent=16.7131, tangentMode=RCTM_User, time=70, value=500}
           <at 90 800>
          ]}}}
       CurveIndex=$User.Speed
      )
      B=1
     )
    )
    Use Spawn Probability=false
   )
   &EmitterState(Enable Distance Culling=true, MaxDistance=2500, Module.bUseMaxDistance=true, Module.bUseMinDistance=false, Scalability Mode=::NewEnumerator1) {module=EmitterState001}
  ]
  particleSpawnScript [
   &InitializeParticle(
    Color=&MakeLinearColor(
     R:NiagaraFloat=1
     G:NiagaraFloat=1
     B:NiagaraFloat=1
     A:NiagaraFloat=&UniformRangedFloat(Minimum=1, Maximum=0.4)
    )
    Lifetime=&UniformRangedFloat(Minimum=0.4, Maximum=0.5)
    Module.Write Color=true
    Module.Write Lifetime=true
    Module.Write Mass=true
    Module.Write Material Random=true
    Module.Write Position=false
    Module.Write Scale=false
    Module.Write SpriteRotation=false
    Module.Write SpriteSize=true
    Position=&SimulationPosition()
    Sprite Rotation=&UniformRangedFloat(Minimum=0, Maximum=360)
    Sprite Size=&Multiply_Vector2(
     A=<0.1 1>
     B=&Vector2DFromFloat(Value=&Multiply_Float(
       A=&FloatFromCurve(
        FloatCurve=DataInterface() {propertyValues {Curve {keys [
            {arriveTangent=0.4214, interpMode=RCIM_Cubic, leaveTangent=0.4225, tangentMode=RCTM_User, time=25, value=0}
            {arriveTangent=7.1685, interpMode=RCIM_Cubic, leaveTangent=7.1685, tangentMode=RCTM_User, time=30, value=50}
            {arriveTangent=7.5, interpMode=RCIM_Cubic, leaveTangent=7.5, tangentMode=RCTM_User, time=69.9874, value=151.101}
            {arriveTangent=10.2726, interpMode=RCIM_Cubic, leaveTangent=10.2726, tangentMode=RCTM_User, time=90.1617, value=288.3149}
           ]}}}
        CurveIndex=$User.Speed
       )
       B=1
      ))
    )
   )
   &SystemLocation(MaskSpawnGroup=false, Offset=<100 0 100>)
   /Niagara/Modules/Spawn/Location/SphereLocation(MaskSpawnGroup=false, Module.Override Randomness=false, Sphere Radius=1000, Surface Only=false)
   &TorusLocation(
    Align Torus With Axis=true
    Handle Radius=400
    Large Radius=1200
    MaskSpawnGroup=false
    Module.Override Randomness=false
    Offset=&Multiply_VectorByFloat(Vector=$Engine.Owner.SystemXAxis, Float=40)
    Torus Axis=<1 0 0>
   )
   &AddVelocity(Velocity=&MakeVector(X=&UniformRangedFloat(Minimum=-150, Maximum=-200), Y=0, Z=0))
   &InheritVelocity(
    Module.Limit Speed=true
    Velocity Limit=1000
    Velocity Scale=&Multiply_VectorByFloat(Vector=<-0.9 -0.9 -0.9>, Float=$User.Speed)
   )
   /SetVariables(Particles.Initial.SpriteSize:Vector2f=&Multiply_Vector2(
     A=<1 1>
     B=&Vector2DFromFloat(Value=&Multiply_Float(
       A=1
       B=&FloatFromCurve(
        FloatCurve=DataInterface() {propertyValues {Curve {keys [
            {arriveTangent=0.4214, interpMode=RCIM_Cubic, leaveTangent=0.4225, tangentMode=RCTM_User, time=25, value=0}
            {arriveTangent=12.969, interpMode=RCIM_Cubic, leaveTangent=12.969, tangentMode=RCTM_User, time=31.9201, value=25.8333}
            {arriveTangent=7.1685, interpMode=RCIM_Cubic, leaveTangent=7.1685, tangentMode=RCTM_User, time=41.1282, value=209.1666}
            {arriveTangent=16.7131, interpMode=RCIM_Cubic, leaveTangent=16.7131, tangentMode=RCTM_User, time=70, value=500}
            <at 90 1500>
           ]}}}
        CurveIndex=$User.Speed
       )
      ))
    )) {isSetParameters=true}
  ]
  particleUpdateScript [
   &ParticleState()
   &Color(
    Color=&ScaleLinearColorByCurve(
     InitialValue=<rgba 1 1 1 0.075>
     ScaleCurve=DataInterface() {propertyValues {
       AlphaCurve {keys [<at 0.0384 1>,<at 0.3631 1>]}
       BlueCurve {keys [<at 0.0281 1>]}
       GreenCurve {keys [<at 0.0281 1>]}
       RedCurve {keys [<at 0.0281 1>]}
      }}
     CurveIndex=$Particles.NormalizedAge
    )
    Scale Alpha=&FloatFromCurve(FloatCurve=DataInterface() {propertyValues {Curve {keys [<at 0 0 RCIM_Cubic>,<at 0.5 1 RCIM_Cubic>,<at 1 0 RCIM_Cubic>]}}})
   ) {enabled=false}
   &ScaleColor(
    Scale Alpha=&FloatFromCurve(FloatCurve=DataInterface() {propertyValues {Curve {keys [<at -0.0046 1.0113>,<at 1.0011 0.002>]}}}, Scale Curve=3)
    ScaleA=true
    ScaleRGB=true
   )
   &ScaleColor(
    Scale Alpha=&FloatFromCurve(FloatCurve=DataInterface() {propertyValues {Curve {keys [<at 0 0>,<at 1 1>]}}}, CurveIndex=$User.Speed, Scale Curve=0.1)
    ScaleA=true
    ScaleRGB=true
   ) {module=ScaleColor001}
   &SolveForcesAndVelocity(Clamp Velocity=false, Limit Acceleration=false)
   &ScaleSpriteSizeBySpeed(Max Scale Factor=<0.5 1>, Min Scale Factor=<0 0>)
   &ScaleSpriteSize(Scale Factor=&Vector2DFromCurve(
     Vector2Curve=DataInterface() {propertyValues {
       XCurve {keys [<at -0.0069 0.5 RCIM_Cubic>,<at 0.4918 1 RCIM_Cubic>,<at 1 0.5 RCIM_Cubic>]}
       YCurve {keys [<at 0 0.0245 RCIM_Cubic>, {arriveTangent=-0.0245, interpMode=RCIM_Cubic, leaveTangent=-0.0245, time=0.4975, value=1}, <at 1 0 RCIM_Cubic>]}
      }}
     ScaleCurve=<1 1>
    ))
   &CameraOffset(Camera Offset Amount=10)
  ]
  rendererClasses [&NiagaraSpriteRendererProperties]
  renderers [{rendererClass=&NiagaraSpriteRendererProperties}]
 }
 FlameOuterStart {
  emitterUpdateScript [
   &EmitterState(Life Cycle Mode=::NewEnumerator1, Loop Duration=0.1, UseLoopDelay=false)
   &SpawnBurst_Instantaneous(Spawn Count=2, Use Spawn Probability=false)
  ]
  particleSpawnScript [
   &InitializeParticle(
    Color=<rgba 1 0.1833 0 0.25>
    Lifetime=0.2
    Mesh Scale=<0.15 1 1>
    Module.Write Color=true
    Module.Write Lifetime=true
    Module.Write Mass=true
    Module.Write Material Random=true
    Module.Write Position=false
    Module.Write Scale=true
    Module.Write SpriteRotation=false
    Module.Write SpriteSize=true
    Position=&SimulationPosition()
    Sprite Size=<5 5>
   )
   /Niagara/Modules/Spawn/Orientation/V2/InitialMeshOrientation(Rotation=<0 0 0.25>, Use Orientation Vector=true, Use Rotation Vector=true)
   &FN_SetSpawnGroup(Emitter Bone Count=2)
  ]
  particleUpdateScript [
   &ParticleState()
   &SampleSkeletalMeshSkeleton(
    Mesh:NiagaraDataInterfaceSkeletalMesh=$User.SkeletalMesh
    Spawn Mode:ENiagaraSkeletalSpawning=::NewEnumerator1
    Specific Socket Index:NiagaraInt32=$Particles.SpawnGroup
   )
   &SkeletalMeshSkeletonLocation(
    Module.Inherit Velocity=false
    Module.Use Sampled Bone as Ribbon ID=true
    Sampled Position Offset:Vector3f=&RotateVectorByQuaternion(Vector To Rotate=<-10 0 0>, Quaternion=$Particles.SampleSkeletalMeshSkeleton.SampledRotation)
   )
   &ScaleMeshSize(Scale Factor=&VectorFromCurve(VectorCurve=DataInterface() {propertyValues {
       XCurve {keys [{arriveTangent=4.0207, interpMode=RCIM_Cubic, leaveTangent=4.0207, tangentMode=RCTM_User, time=0, value=0}, <at 0.9989 2 RCIM_Cubic>]}
       YCurve {keys [
         {arriveTangent=10.0362, interpMode=RCIM_Cubic, leaveTangent=10.0362, tangentMode=RCTM_User, time=0, value=0}
         {interpMode=RCIM_Cubic, tangentMode=RCTM_User, time=0.19, value=1}
         {arriveTangent=-3.6354, interpMode=RCIM_Cubic, leaveTangent=-3.6354, tangentMode=RCTM_User, time=1, value=0}
        ]}
       ZCurve {keys [
         {arriveTangent=10.128, interpMode=RCIM_Cubic, leaveTangent=10.128, tangentMode=RCTM_User, time=0, value=0}
         {interpMode=RCIM_Cubic, tangentMode=RCTM_User, time=0.19, value=1}
         {arriveTangent=-4.3855, interpMode=RCIM_Cubic, leaveTangent=-4.3855, tangentMode=RCTM_User, time=1, value=0}
        ]}
      }}))
   &SolveForcesAndVelocity(Clamp Velocity=false, Limit Acceleration=false)
   &DynamicMaterialParameters(
    Index 0 Param 1=&FloatFromCurve(FloatCurve=DataInterface() {propertyValues {Curve {keys [{arriveTangent=-2.9083, interpMode=RCIM_Cubic, leaveTangent=-2.9083, tangentMode=RCTM_User, time=0, value=1}, <at 1 0 RCIM_Cubic>]}}})
    Index 0 Param 2=5
    Index 0 Param 3=20
    Module.Param0WriteEnabled=true
    Module.Param1WriteEnabled=true
    Module.Param2WriteEnabled=true
    Module.Param3WriteEnabled=true
   )
   &ScaleColor(
    Scale Alpha=$User.BoostReady
    Scale RGB=&VectorFromFloat(Value=&FloatFromCurve(FloatCurve=DataInterface() {propertyValues {Curve {keys [<at 0 1 RCIM_Cubic>,<at 0.25 7 RCIM_Cubic>,<at 1 1 RCIM_Cubic>]}}}))
    ScaleA=true
    ScaleRGB=true
   )
  ]
  rendererClasses [&NiagaraMeshRendererProperties]
  renderers [{rendererClass=&NiagaraMeshRendererProperties, rendererIndex=0}]
 }
}
summary {
 systemName=NS_Hoagie_Idle
 emitters [
  {emitterName=Boost_Ready, rendererClasses [&NiagaraSpriteRendererProperties, &NiagaraLightRendererProperties]}
  {emitterName=FlameOuterStart, rendererClasses [&NiagaraMeshRendererProperties]}
  {emitterName=FN_Sprite_SpeedLines001, rendererClasses [&NiagaraSpriteRendererProperties]}
 ]
 userVariables [
  User.BoostReady {type=NiagaraFloat, defaultValue=1, description=""}
  User.isCritical {type=NiagaraBool}
  User.SkeletalMesh {
   type=NiagaraDataInterfaceSkeletalMesh
   defaultValue {struct=NiagaraExt_VariableValue_DataInterface, value {dataInterface=&self:NiagaraDataInterfaceSkeletalMesh_0}}
  }
  User.Speed {type=NiagaraFloat, defaultValue=0, description=""}
 ]
}
