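# Niagara effect for landing skids scraping a surface, with sprite Smoke and Sparks emitters driven by User.Speed and thrown along User.ThrustVelocity.
# Smoke spawns in small bursts with probability set by a physical-material comparison, then drifts with drag, expands and fades.
# Sparks stream at a speed-curved rate without an active material check, falling under gravity and curl noise.
# Speed curves also scale both emitters' sprite sizes.
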NiagaraSystem /Hoagie/FX/NS_Hoagie_Scrape_LandingSkids
dependencies {usedDataInterfaces [&NiagaraDataInterfaceCurve], usedDynamicInputs [], usedModules [&SystemState]}
emitters {
 Smoke {
  emitterUpdateScript [
   &EmitterState()
   &SpawnBurst_Instantaneous(
    Spawn Count=&UniformRangedInt(
     Minimum=&UniformRangedInt(Minimum=1, Maximum=2)
     Maximum=&MakeIntFromFloat(Float=&FloatFromCurve(
       FloatCurve=DataInterface() {propertyValues {Curve {keys [
           {arriveTangent=-0.0078, interpMode=RCIM_Cubic, leaveTangent=-0.0078, tangentMode=RCTM_User, time=65, value=0}
           {arriveTangent=0.1798, interpMode=RCIM_Cubic, leaveTangent=0.1798, time=65.896, value=1}
           <at 70.5614 1 RCIM_Cubic>
           <at 89.4969 1 RCIM_Cubic>
          ]}}}
       CurveIndex=$User.Speed
      ))
    )
    Spawn Probability=/Hoagie/FX/DynamicInputs/FN_FloatFromPhysicalMaterialComparison(Index:NiagaraInt32=0, Physical Material Index:NiagaraInt32=$User.PhysMatIndex)
    Use Spawn Probability=true
   )
   &SpawnBurst_Instantaneous(
    Spawn Count=&UniformRangedInt(Minimum=5, Maximum=10)
    Spawn Probability=&FN_FloatFromPhysicalMaterialComparison(Index:NiagaraInt32=0, Physical Material Index:NiagaraInt32=$User.PhysMatIndex)
    Use Spawn Probability=true
   ) {enabled=false, module=SpawnBurst_Instantaneous001}
  ]
  particleSpawnScript [
   /Niagara/Modules/Spawn/Initialization/InitializeParticle(
    Color=<rgba 0.5 0.5677 0.5677 0.125>
    Lifetime=&UniformRangedFloat(Minimum=0.2, Maximum=0.5)
    Module.Write Color=true
    Module.Write Lifetime=true
    Module.Write Mass=true
    Module.Write Material Random=true
    Module.Write Position=false
    Module.Write Scale=false
    Module.Write SpriteRotation=true
    Module.Write SpriteSize=true
    Position=&SimulationPosition()
    Sprite Rotation=&UniformRangedFloat(Minimum=0, Maximum=360)
    Sprite Size=&Vector2DFromFloat(Value=&UniformRangedFloat(Minimum=100, Maximum=250))
   )
   &BoxLocation(Box Size=<15 50 15>, MaskSpawnGroup=false, Module.Override Seed=false, Rotate Box=false, Surface Only=false)
   /Niagara/Modules/Spawn/Location/SphereLocation(Hemisphere X=true, MaskSpawnGroup=false, Module.Override Randomness=false, Sphere Radius=50, Surface Only=false)
   &AddVelocity(Velocity=&Multiply_Vector(
     A=$Output.SphereLocation.SphereNormal
     B=&VectorFromFloat(Value=&UniformRangedFloat(Minimum=0, Maximum=200))
    ))
   &AddVelocity(Velocity=&Multiply_Vector(
     A=&NormalizeVector(Vector To Normalize=$User.ThrustVelocity)
     B=&VectorFromFloat(Value=&UniformRangedFloat(Minimum=200, Maximum=300))
    )) {module=AddVelocity001}
   &ApplyInitialForces()
   &CameraOffset(Camera Offset Amount=50)
  ]
  particleUpdateScript [
   &ParticleState()
   &AccelerationForce()
   &Drag(Drag=2, Use Linear Drag=true, Use Rotational Drag=false)
   &SolveForcesAndVelocity(Clamp Velocity=false, Limit Acceleration=false)
   &ScaleColor(
    Scale Alpha=&FloatFromCurve(FloatCurve=DataInterface() {propertyValues {Curve {keys [
         {interpMode=RCIM_Cubic, tangentMode=RCTM_User, time=0, value=0}
         {arriveTangent=0.002, interpMode=RCIM_Cubic, leaveTangent=0.002, tangentMode=RCTM_User, time=0.25, value=1}
         {interpMode=RCIM_Cubic, tangentMode=RCTM_User, time=1.0011, value=0.002}
        ]}}})
    Scale RGB=&Lerp_Vector(A=<1 1 1>, B=<1 1 1>, Alpha=$Particles.NormalizedAge)
    ScaleA=true
    ScaleRGB=true
   )
   &DynamicMaterialParameters(
    Index 0 Param 1=&FloatFromCurve(FloatCurve=DataInterface() {propertyValues {Curve {keys [<at 0.38 1.0113 RCIM_Cubic>,<at 1.0011 0.002 RCIM_Cubic>]}}})
    Index 0 Param 4=1
    Module.Param0WriteEnabled=true
    Module.Param1WriteEnabled=true
    Module.Param2WriteEnabled=true
    Module.Param3WriteEnabled=true
   )
   &ScaleSpriteSize(Scale Factor=&Vector2DFromFloat(Value=&FloatFromCurve(
      FloatCurve=DataInterface() {propertyValues {Curve {keys [
          {arriveTangent=0.4214, interpMode=RCIM_Cubic, time=25.4285, value=1}
          {arriveTangent=0.0224, interpMode=RCIM_Cubic, leaveTangent=0.0224, time=30.5524, value=1.4956}
          {arriveTangent=0.0085, interpMode=RCIM_Cubic, leaveTangent=0.0085, time=70, value=2}
          <at 90 2 RCIM_Cubic>
         ]}}}
      CurveIndex=$User.Speed
     ))) {module=ScaleSpriteSize001}
   &ScaleSpriteSize(Scale Factor=&Vector2DFromFloat(Value=&FloatFromCurve(FloatCurve=DataInterface() {propertyValues {Curve {keys [<at 0 1.2 RCIM_Cubic>,<at 1.0011 3.48 RCIM_Cubic>]}}})))
  ]
  rendererClasses [&NiagaraSpriteRendererProperties]
  renderers [{rendererClass=&NiagaraSpriteRendererProperties}]
 }
 Sparks {
  emitterUpdateScript [
   &EmitterState()
   &SpawnBurst_Instantaneous(
    Spawn Count=&UniformRangedInt(Minimum=16, Maximum=32)
    Spawn Probability=&FN_FloatFromPhysicalMaterialComparison(Index:NiagaraInt32=0, Physical Material Index:NiagaraInt32=$User.PhysMatIndex)
    Use Spawn Probability=true
   ) {enabled=false}
   &SpawnRate(
    SpawnRate=&FloatFromCurve(
     FloatCurve=DataInterface() {propertyValues {Curve {keys [
         {arriveTangent=-0.0078, interpMode=RCIM_Cubic, leaveTangent=-0.0078, tangentMode=RCTM_User, time=25, value=0}
         <at 29.3003 0.6081>
         {arriveTangent=0.0541, interpMode=RCIM_Cubic, leaveTangent=0.0541, time=41.6721, value=0.9785}
         {arriveTangent=0.0445, interpMode=RCIM_Cubic, leaveTangent=0.0445, time=66.2601, value=2.6064}
         <at 89.4969 3.108 RCIM_Cubic>
        ]}}}
     CurveIndex=$User.Speed
     Scale Curve=150
    )
    Use Spawn Probability=false
   )
  ]
  particleSpawnScript [
   &SystemLocation(MaskSpawnGroup=false, Offset=&UniformRangedVector(Minimum=<-10 -10 0>, Maximum=<10 10 0>))
   &InitializeParticle(
    Color=<rgba 50 10 1 1>
    Lifetime=&UniformRangedFloat(Minimum=0.25, Maximum=0.65)
    Module.Write Color=true
    Module.Write Lifetime=true
    Module.Write Mass=true
    Module.Write Material Random=true
    Module.Write Position=false
    Module.Write Scale=false
    Module.Write SpriteRotation=false
    Module.Write SpriteSize=true
    Position=&SimulationPosition()
    Sprite Size=&UniformRangedVector2D(Minimum=<1 6>, Maximum=<2 12>)
   )
   &BoxLocation(Box Size=<25 25 200>, MaskSpawnGroup=false, Module.Override Seed=false, Rotate Box=false, Surface Only=false)
   &SphereLocation(
    Hemisphere X=true
    MaskSpawnGroup=false
    Module.Override Randomness=false
    Non Uniform Scale=<0.25 0.5 1>
    Sphere Radius=15
    Surface Only=false
   ) {enabled=false}
   &AddVelocity(Velocity=&Multiply_Vector(
     A=&NormalizeVector(Vector To Normalize=$User.ThrustVelocity)
     B=&VectorFromFloat(Value=&UniformRangedFloat(Minimum=200, Maximum=1000))
    ))
   &AddVelocity(Velocity=&UniformRangedVector(Minimum=<0 0 0>, Maximum=<0 0 100>)) {module=AddVelocity001}
   &CameraOffset(Camera Offset Amount=50)
   &AddVelocityInCone(
    Cone Angle=15
    Cone Axis=/Niagara/DynamicInputs/Random/RandomVector(
     Vector Scale:NiagaraFloat=15
     Constraint Vector:Vector3f=<0 0 1>
     Randomness Mode:ENiagaraRandomnessMode=::NewEnumerator0
     Random Seed:NiagaraInt32=0
     Fixed Random Seed:NiagaraBool=false
    )
    Module.Use Velocity Falloff On Cone Axis=false
    Velocity Strength=500
   )
  ]
  particleUpdateScript [
   &ParticleState()
   &AccelerationForce(Acceleration=<0 0 -980>)
   /Niagara/Modules/Update/Forces/CurlNoiseForce(Noise Frequency=0.25, Noise Strength=100, Pan Noise=true, Pan Noise Field=<1 0 0>)
   &SolveForcesAndVelocity(Clamp Velocity=false, Limit Acceleration=false)
   &ScaleColor(
    Scale RGB=&Lerp_Vector(A=<1 1 1>, B=<0.3333 0.026 0.1205>, Alpha=$Particles.NormalizedAge)
    ScaleA=true
    ScaleRGB=true
   )
   &ScaleSpriteSize(Scale Factor=&Vector2DFromFloat(Value=&FloatFromCurve(
      FloatCurve=DataInterface() {propertyValues {Curve {keys [
          {arriveTangent=0.4214, interpMode=RCIM_Cubic, time=25, value=0}
          {arriveTangent=0.0444, interpMode=RCIM_Cubic, leaveTangent=0.0444, time=40, value=2}
          <at 70 2 RCIM_Cubic>
          <at 90 2 RCIM_Cubic>
         ]}}}
      CurveIndex=$User.Speed
     )))
   &ScaleSpriteSizeBySpeed(Max Scale Factor=<1 2>, Min Scale Factor=<1 0.25>)
  ]
  rendererClasses [&NiagaraSpriteRendererProperties]
  renderers [{rendererClass=&NiagaraSpriteRendererProperties}]
 }
}
summary {
 systemName=NS_Hoagie_Scrape_LandingSkids
 emitters [
  {emitterName=Sparks, rendererClasses [&NiagaraSpriteRendererProperties]}
  {emitterName=Smoke, rendererClasses [&NiagaraSpriteRendererProperties]}
 ]
 userVariables [
  User.PhysMatIndex {type=NiagaraInt32}
  User.PhysMatInt {type=NiagaraInt32}
  User.Speed {type=NiagaraFloat, defaultValue=0, description=""}
  User.ThrustVelocity {type=Vector3f, description="", defaultValue {struct=Vector3f, value=<0 0 0>}}
 ]
}
