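; Helicopter vehicle Blueprint (Hoagie variant) handling flight cosmetics, seat changes, rotor hits, boost and damage feedback.
; Native vehicle events drive the anim Blueprint, audio controller, Niagara effects, material instances, camera shakes and controller rumble.
; Rotor hits apply GameplayEffects to buildings and the helicopter, and launch pawns on the server.
; Unusually, honking spawns SilverBlazer consumable actors, and boosting fires two projectiles from the driver's position.
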(event Custom|SetupNative
  ;[1 1 1] Setup Native
  (Variables|Vehicle|SetHoagieIdleFX_Native (Variables|Default|GetHoagie%20Idle%20FX))
  (Variables|Vehicle|SetRotorWashFX_Native (Variables|Default|GetRotorWashFX))
  (Variables|Vehicle|SetDamageFX_Native (Variables|Default|GetDamageFX))
  (bind _audiocontroller (Utilities|GetAudioController)
    (:Success
      (Class|BPHoagieAudioController|CacheReferences _audiocontroller 0))))

(event Custom|LocalPassengerADS
  (params (AimDownSights :type Boolean))
  (bind _mid_slotzero (Variables|Default|GetMID_SlotZero))
  ;[1 1 1] Set Player Is ADS bool
  (bind _output_get (Variables|Default|SetPlayerIsADS AimDownSights))
  (if _output_get
    (Rendering|Material|SetScalarParameterValue _mid_slotzero "PlayerIsADS" 1.0 :!class /Script/Engine.MaterialInstanceDynamic)
    (else
      (Rendering|Material|SetScalarParameterValue _mid_slotzero "PlayerIsADS" 0.0 :!class /Script/Engine.MaterialInstanceDynamic))))

(event Custom|SpawnCritFX
  ;[1 1 1] Weakpoint Crit FX
  (Effects|SpawnFXSystemAttached :FXSystemAsset (Variables|FX|GetFX_CritHit) :AttachToComponent (Variables|Vehicle|GetSkeletalMesh) :AttachPointName (select (Variables|Default|GetHeliRotorSocket) "MainRotor" "TailRotor") :Scale "1.000000,1.000000,1.000000" :LocationType "SnapToTarget" :bAutoDestroy true :PoolingMethod "None" :bAutoActivate true))

(event EventOnCritRotor
  (params (Damage :type Float%20%28single-precision%29) (ImpactLocation :type Vector%20%28by%20ref%29 :default "0, 0, 0") (DamageInstigator :type Controller%20Object%20Reference) (DamageCauser :type Actor%20Object%20Reference) (bMainRotor :type Boolean))
  ;[1 1 1] Weakpoint Crits
  (Variables|Default|SetHeliRotorSocket bMainRotor)
  (Utilities|Casting|CastToPlayerController DamageInstigator)
  (CallFunction|SpawnCritFX))

(event EventSetRotorWashActive
  (params (bActive :type Boolean))
  ;[1 1 1] EVENT -> Set Rotor Wash Active
  (Variables|Default|SetIsRotorWashActive bActive)
  (Development|PrintString "ROTORWASH ACTIVE"))

(event EventK2_ApplyCosmeticWrap
  (params (LoadedWrap :type Athena%20Item%20Wrap%20Definition%20Object%20Reference))
  (bind _islocalplayerpassenger (Variables|Default|GetIsLocalPlayerPassenger))
  ;[1 1 1] Apply Cosmetic Wrap
  (if (not (GameplayTags|HasMatchingGameplayTag "(TagName=\"Gameplay.Status.Chrome\")"))
    (Class|FortVehicleInterface|ApplyMaterialandUseWrapifPossible (select _islocalplayerpassenger "/Hoagie/Materials/MI_Hoagie_Parts_Masked.MI_Hoagie_Parts_Masked" "/Hoagie/Materials/MI_Hoagie_Parts.MI_Hoagie_Parts") 0 (delegate Custom|OnMIDCreated_SlotZero))
    (Class|FortVehicleInterface|ApplyMaterialandUseWrapifPossible (select _islocalplayerpassenger "/Hoagie/Materials/MI_Hoagie_Body_Masked.MI_Hoagie_Body_Masked" "/Hoagie/Materials/MI_Hoagie_Body.MI_Hoagie_Body") 2 (delegate Custom|OnMIDCreated_Slot2))
    (CallFunction|UpdateDamageState (Variables|Default|GetDamageValue))))

(event EventSetTailRotorRotation
  (params (Degrees :type Float%20%28single-precision%29))
  ;[1 1 1] EVENT -> Set Rotor Rotation
  (Variables|/Script/HoagieRuntime.FortHoagieVehicleAnimInstance|SetTailRotorRotaton_Roll (Math|Rotator|MakeRotator 0.0 Degrees) (Variables|Animation|GetHoagieAnimBP)))

(event Custom|LocalPassengerMatSwap
  (bind _islocalplayerpassenger (Variables|Default|GetIsLocalPlayerPassenger))
  ;[1 1 1] Apply Cosmetic Wrap
  (if (not (GameplayTags|HasMatchingGameplayTag "(TagName=\"Gameplay.Status.Chrome\")"))
    (Class|FortVehicleInterface|ApplyMaterialandUseWrapifPossible (select _islocalplayerpassenger "/Hoagie/Materials/MI_Hoagie_Parts_Masked.MI_Hoagie_Parts_Masked" "/Hoagie/Materials/MI_Hoagie_Parts.MI_Hoagie_Parts") 0 (delegate Custom|OnMIDCreated_SlotZero))
    (Class|FortVehicleInterface|ApplyMaterialandUseWrapifPossible (select _islocalplayerpassenger "/Hoagie/Materials/MI_Hoagie_Body_Masked.MI_Hoagie_Body_Masked" "/Hoagie/Materials/MI_Hoagie_Body.MI_Hoagie_Body") 2 (delegate Custom|OnMIDCreated_Slot2))
    (CallFunction|UpdateDamageState (Variables|Default|GetDamageValue))))

(event EventSetMainRotorRotation
  (params (Degrees :type Float%20%28single-precision%29))
  ;[1 1 1] EVENT -> Set Rotor Rotation
  (Variables|/Script/HoagieRuntime.FortHoagieVehicleAnimInstance|SetMainRotorRotation_Roll (Math|Rotator|MakeRotator 0.0 0.0 Degrees) (Variables|Animation|GetHoagieAnimBP)))

(event Custom|OnMIDCreated_SlotZero
  (params (MID :type Material%20Instance%20Dynamic%20Object%20Reference))
  ;[1 1 1] Apply Cosmetic Wrap
  (Variables|Default|SetMID_SlotZero MID))

(event Custom|OnMIDCreated_Slot2
  (params (MID :type Material%20Instance%20Dynamic%20Object%20Reference))
  ;[1 1 1] Apply Cosmetic Wrap
  (Variables|Default|SetMID_Slot2 MID))

(event Vehicle|EventOnHonk
  (bind _returnvalue (Vehicle|GetDriver))
  (bind _rootcomponent (Variables|/Script/Engine.Actor|GetRootComponent _returnvalue))
  (bind _returnvalue_2 (Transformation|GetRelativeTransform _rootcomponent))
  ;[1 1 1] EVENT -> Honk
  (bind _audiocontroller (Utilities|GetAudioController)
    (:Success
      (Class|BPHoagieAudioController|OnHornButtonPressed _audiocontroller)
      (Projectile|SpawnProjectile "/Game/Athena/Items/Consumables/SilverBlazer/B_Prj_Athena_SilverBlazer_V2_.B_Prj_Athena_SilverBlazer_V2__C" 0 (Math|Transform|TransformLocation _returnvalue_2) "0, 0, 0" (Math|Transform|TransformRotation _returnvalue_2)))
    (:Failed
      (bind _returnvalue_3 (Building|SpawnBuildingGameplayActor "/Game/Athena/Items/Consumables/SilverBlazer/BGA_Athena_SilverBlazerCore.BGA_Athena_SilverBlazerCore_C" _returnvalue_2 _returnvalue))
      (Transformation|AttachActorToComponent _returnvalue_3 (Variables|Default|GetTailRotorAssist) "None" "SnapToTarget" "KeepRelative" "KeepRelative" true))))

(event EventBeginPlay
  ;[1 1 1] BEGIN PLAY
  (Class|FortAsyncActionCurrentPlaylistReady|PlaylistDataReadyAsync
    (:OnReady
      (CallFunction|CacheHotfixValues)
      (CallFunction|SetupNative)
      (Collision|AssignOnDamageablePartHealthChanged (delegate Custom|OnDamageablePartHealthChanged_Event))
      (Utilities|FlowControl|Sequence
        (:then_0
          (bind _assk_hoagie_protomesh_anim_bp (Utilities|Casting|CastToSK_Hoagie_Protomesh_AnimBP (Components|SkeletalMesh|GetAnimInstance (Variables|Vehicle|GetSkeletalMesh))))
          (Variables|Animation|SetHoagieAnimBP _assk_hoagie_protomesh_anim_bp))
        (:then_1
          (CallFunction|EventonHealthChanged)
          (Health|BindEventtoOnHealthChanged (EventDispatchers|CreateEvent :!event Custom|EventOnHealthChanged))
          (CallFunction|SetDamageStateThresholds)
          (Effects|Components|ParticleSystem|SetFloatParameter (Variables|Default|GetHoagie%20Idle%20FX) "User.BoostReady" 0.0) ; Don't start with boost on
          (AimAssist|RegisterAimAssistTarget self (Variables|Default|GetTailRotorAssist))
          (Effects|Components|ParticleSystem|SetBoolParameter (Variables|Default|GetDamageFX) "User.isLow" (TouchPlatform|IsMobileGame))
          (Vehicle|Destruction|BindEventtoOnVehicleExplodeDelegate (EventDispatchers|CreateEvent :!event Custom|HandleOnExplode))
          (Vehicle|Destruction|BindEventtoOnStartCriticalHealth (EventDispatchers|CreateEvent :!event Custom|HandleOnStartCriticalHealth)))))))

(event Custom|OnDamageablePartHealthChanged_Event
  (params (PartConfig :type Vehicle%20Damageable%20Part%20Config%20Structure%20%28by%20ref%29) (Part :type Vehicle%20Damageable%20Part%20Structure%20%28by%20ref%29) (bNewlyDisabled :type Boolean) (bNewlyEnabled :type Boolean))
  ;[1 1 1] BEGIN PLAY
  (if bNewlyDisabled
    (CallFunction|GlassDestroyed (Utilities|Struct|BreakVehicleDamageablePartConfig PartConfig))))

(event Custom|HandleOnStartCriticalHealth
  (params (LastDamageInstigator :type Controller%20Object%20Reference))
  (bind _pilotcontroller (Variables|Default|GetPilotController))
  ;[1 0.6859 0.7158] HandleOnStartCriticalHealth: Called when the health of the vehicle reaches critical value (0.f)
  (Networking|SwitchHasAuthority
    (:Authority
      (Variables|Default|SetIsCritical true)
      (CallFunction|StartCurieFire))
    (:Remote
      (Utilities|FlowControl|Sequence
        (:then_0
          (CallFunction|StopPassiveRumble)
          (CallFunction|CriticalRumbleStart))
        (:then_1
          (Utilities|IsValid _pilotcontroller
            (:Is%20Valid
              (bind _playercameramanager (Variables|/Script/Engine.PlayerController|GetPlayerCameraManager _pilotcontroller)
                (:then
                  (bind _returnvalue (CameraShakes|StartCameraShake :self _playercameramanager :ShakeClass (Variables|CameraShake|GetCameraShakeCritical) :Scale 1.0 :PlaySpace "CameraLocal"))
                  (Variables|CameraShake|SetCriticalCameraShake (CameraShakes|ConvLegacyCameraShake _returnvalue)))
                (:else)))))))))

(event Custom|CriticalRumbleStart
  ;[1 0.6859 0.7158] HandleOnStartCriticalHealth: Called when the health of the vehicle reaches critical value (0.f)
  (Utilities|FlowControl|Sequence
    (:then_0
      (Game|Feedback|PlayDynamicForceFeedback (Variables|Default|GetPilotController) (stale _rumbleintensity) -1 true true true true))
    (:then_1
      (bind (_direction _rumbleintensity) (Timeline|CriticalRumble PlayFromStart)
        (:Update
          _)
        (:Finished
          _)))))

(event Custom|StopPilotCameraShake
  ;[1 1 1] Camera Shake - Stop Pilot Camera Shakes
  (Utilities|IsValid (Variables|Default|GetPilot)
    (:Is%20Valid
      (Utilities|FlowControl|Sequence
        (:then_0
          (Utilities|IsValid (Variables|CameraShake|GetFlightCameraShake)
            (:Is%20Valid
              (CameraShakes|StopCameraShake (Variables|/Script/Engine.PlayerController|GetPlayerCameraManager (Fortnite|GetFortPlayerControllerfromActor (Variables|Default|GetPilot))) (Variables|CameraShake|GetFlightCameraShake) false)
              (Variables|CameraShake|SetFlightCameraShake 0))))
        (:then_1
          (Utilities|IsValid (Variables|CameraShake|GetSpeedCameraShake)
            (:Is%20Valid
              (CameraShakes|StopCameraShake (Variables|/Script/Engine.PlayerController|GetPlayerCameraManager (Fortnite|GetFortPlayerControllerfromActor (Variables|Default|GetPilot))) (Variables|CameraShake|GetSpeedCameraShake) false)
              (Variables|CameraShake|SetSpeedCameraShake 0))))
        (:then_2
          (Utilities|IsValid (Variables|CameraShake|GetBoostCameraShake)
            (:Is%20Valid
              (CameraShakes|StopCameraShake (Variables|/Script/Engine.PlayerController|GetPlayerCameraManager (Fortnite|GetFortPlayerControllerfromActor (Variables|Default|GetPilot))) (Variables|CameraShake|GetBoostCameraShake) false)
              (Variables|CameraShake|SetBoostCameraShake 0))))))))

(event EventOnHoldExitStarted
  (params (ExitingPawn :type Fort%20Pawn%20Object%20Reference) (ExitDuration :type Float%20%28single-precision%29))
  ;[1 1 1] EVENT -> OnHoldExitStarted/Stopped
  (Ability|SendGameplayEventtoActor ExitingPawn "(TagName=\"Athena.Vehicle.Biplane.OnHoldExitStarted\")" (Utilities|Struct|MakeGameplayEventData :ContextHandle "()" :InstigatorTags "(GameplayTags=)" :TargetTags "(GameplayTags=)" :EventMagnitude ExitDuration :TargetData "()")))

(event Custom|CriticalRumbleStop
  ;[1 0.6859 0.7158] HandleOnStartCriticalHealth: Called when the health of the vehicle reaches critical value (0.f)
  (Utilities|FlowControl|Sequence
    (:then_0
      (Timeline|CriticalRumble Stop))
    (:then_1
      (Game|Feedback|PlayDynamicForceFeedback (Variables|Default|GetPilotController) (stale Custom|CriticalRumbleStart _rumbleintensity) -1 true true true true))))

(event EventOnHoldExitStopped
  (params (ExitingPawn :type Fort%20Pawn%20Object%20Reference))
  ;[1 1 1] EVENT -> OnHoldExitStarted/Stopped
  (Ability|SendGameplayEventtoActor ExitingPawn "(TagName=\"Athena.Vehicle.Biplane.OnHoldExitStopped\")"))

(event Custom|StartDriverCameraShake
  (bind _returnvalue (Vehicle|GetDriver))
  (bind _returnvalue_2 (Fortnite|GetFortPlayerControllerfromActor _returnvalue))
  (bind _playercameramanager (Variables|/Script/Engine.PlayerController|GetPlayerCameraManager _returnvalue_2))
  ;[1 1 1] Camera Shake - Loc in Flight Movement
  (Utilities|IsValid _returnvalue_2
    (:Is%20Valid
      (bind _returnvalue_3 (CameraShakes|StartCameraShake :self _playercameramanager :Scale 0.01 :PlaySpace "CameraLocal"))
      (Variables|CameraShake|SetFlightCameraShake (CameraShakes|ConvLegacyCameraShake _returnvalue_3))
      ;[1 1 1] Start Speed Based Camera Shake
      (bind _returnvalue_4 (CameraShakes|StartCameraShake :self _playercameramanager :Scale 0.01 :PlaySpace "CameraLocal"))
      (Variables|CameraShake|SetSpeedCameraShake (CameraShakes|ConvLegacyCameraShake _returnvalue_4)))))

(event EventOnRotorsStop
  (bind _hoagie_idle_fx (Variables|Default|GetHoagie%20Idle%20FX))
  ;[1 1 1] EVENT - Rotors Stop
  (Effects|Components|ParticleSystem|SetFloatParameter _hoagie_idle_fx "User.BoostReady" 0.0)
  (Components|Activation|Deactivate _hoagie_idle_fx)
  (Components|Activation|Deactivate (Variables|Default|GetRotorWashFX))
  (CallFunction|RunningLIghtsOff))

(event Custom|EventStartUp
  ;[0.75 1 0.9062] StartUp Sequence
  (bind _audiocontroller (Utilities|GetAudioController)
    (:Success
      (Class|BPHoagieAudioController|OnNotifyPilotEnter _audiocontroller))
    (:Failed
      _))
  (Variables|/Hoagie/Animations/SK_Hoagie_Protomesh_AnimBP.SK_Hoagie_Protomesh_AnimBP_C|SetIsStarted true (Variables|Animation|GetHoagieAnimBP))
  (Components|Activation|Activate (Variables|Default|GetHoagie%20Idle%20FX) false)
  (Components|Activation|Activate (Variables|Default|GetRotorWashFX) false)
  (CallFunction|RunningLIghtsOn))

(event Custom|Running%20LIghts%20On)

(event EventOnPawnEnterVehicle
  (params (PlayerPawn :type Fort%20Player%20Pawn%20Object%20Reference) (SeatIdx :type Integer))
  ;[0.06191 0.5156 0] Pawn Enter Vehicle
  (CallFunction|ApplyBoostCooldown PlayerPawn) ; Mobile Boost Radial
  (if (== SeatIdx 0)
    (CallFunction|ManagePilotEnter PlayerPawn)
    (CallFunction|EventStartUp)
    (Variables|Default|SetIsStarting true)
    (elif (>= SeatIdx 1)
      (Utilities|Array|AddUnique (Variables|Default|GetPassengerPawns) PlayerPawn)
      (CallFunction|ManagePassengerEnter PlayerPawn))))

(event Custom|On%20Rotor%20Damage%20Dealt%20Outer%20FX
  (params (Impact%20Point :type Vector) (Rotation :type Rotator))
  ;[1 1 1] Rotor Impact - Outer
  (Networking|CanExecuteCosmeticEvents
    (:True
      (Effects|Components|ParticleSystem|SpawnFXAtLocation :FXSystemAsset (Variables|FX|GetFX_RotorScrapes) :Location Impact%20Point :Scale "1.000000,1.000000,1.000000" :bAutoDestroy false :PoolingMethod "AutoRelease" :bAutoActivate true)
      (CallFunction|RotorHitHaptics)
      (bind _audiocontroller (Utilities|GetAudioController)
        (:Success
          (Class|BPHoagieAudioController|OnNotifyRotorOuterImpact _audiocontroller))))))

(event EventOnRotorDamageDealtOuter
  (params (Impact :type Hit%20Result%20Structure%20%28by%20ref%29))
  (bind (_bblockinghit _binitialoverlap _time _distance _location _impactpoint _normal _impactnormal _physmat _hitactor) (Collision|BreakHitResult Impact))
  (bind _returnvalue (Vehicle|GetDriver))
  ;[1 1 1] Rotor Impact - Outer
  (bind _asfort_aipawn (Utilities|Casting|CastToFortAIPawn _hitactor)
    (:then
      (if (Networking|IsServer)
        (CallFunction|RotorDamagePawns _asfort_aipawn)
        (else
          _)))
    (:CastFailed
      _))
  (CallFunction|OnRotorDamageDealtOuterFX _impactpoint (Math|Rotator|CombineRotators (Math|Rotator|MakeRotfromX (- _impactpoint (Math|Vector|MakeVector (.x (Transformation|GetWorldLocation (Variables|Vehicle|GetSkeletalMesh))) (.y (Transformation|GetWorldLocation (Variables|Vehicle|GetSkeletalMesh))) (.z _impactpoint)))) "0,-90.000000, 0"))
  (Utilities|IsValid _hitactor
    (:Is%20Valid
      (Utilities|IsValid _returnvalue
        (:Is%20Valid
          (GameplayEffects|ApplyGameplayEffectToTarget (Ability|GetAbilitySystemComponent _returnvalue) (Variables|GameplayEffects|GetGE_RotorHit_DamageBuilding_Outer) (Ability|GetAbilitySystemComponent _hitactor) 0.0 (GameplayEffects|MakeEffectContext (Ability|GetAbilitySystemComponent _returnvalue))) ; Damage Building
          ;;[1 1 1] Damage Self
          (bind _asbuilding_prop (Utilities|Casting|CastToBuildingProp _hitactor)
            (:then
              (if (CallFunction|ShouldDamageBuilding _asbuilding_prop 125.0)
                _
                (else
                  _)))
            (:CastFailed
              _))
          (GameplayEffects|ApplyGameplayEffectToSelf (Vehicle|GetVehicleAbilitySystemComponent) (Variables|GameplayEffects|GetGE_RotorHit_DamageSelf_Outer) 0.0 (GameplayEffects|MakeEffectContext (Vehicle|GetVehicleAbilitySystemComponent))))))))

(event Custom|Running%20LIghts%20Off)

(event EventOnPawnExitVehicle
  (params (PlayerPawn :type Fort%20Player%20Pawn%20Object%20Reference) (ExitSocketName :type Name :default "None"))
  (bind (_returnvalue_x _returnvalue_y) (Math|Vector|BreakVector (* (Transformation|GetVelocity) (Variables|Default|GetExitLaunchMultiplier))))
  (bind _lastdamagecauser (Variables|Default|GetLastDamageCauser))
  ;[0.1927 0 0.000944] Pawn Exit Vehicle
  ;; ApplyExitGE
  (GameplayEffects|ApplyGameplayEffectToTarget (Ability|GetAbilitySystemComponent self) (Variables|GameplayEffects|GetGE_RecentlyExit) (Ability|GetAbilitySystemComponent PlayerPawn) 1.0 (GameplayEffects|MakeEffectContext (Ability|GetAbilitySystemComponent self)))
  (CallFunction|RemoveBoostCooldownEvent PlayerPawn) ; Mobile Boost Radial
  (if (not (Vehicle|IsinVehicle PlayerPawn))
    (Pawn|Character|LaunchCharacterJump PlayerPawn (Math|Vector|MakeVector _returnvalue_x _returnvalue_y) false false false true) ; Launch players in direction of Heli's movement
  )
  (Utilities|FlowControl|Sequence
    (:then_0
      (if (Vehicle|Destruction|IsCriticalHealth)
        (Utilities|IsValid _lastdamagecauser
          (:Is%20Valid
            (Class|FortPlayerPawnAthena|SetFallInstigator :self (Utilities|Casting|CastToFortPlayerPawnAthena PlayerPawn) :NewFallInstigator (Game|GetInstigatorController _lastdamagecauser))))))
    (:then_1
      (if (Utilities|Equal(Object) (Variables|Default|GetPilot) PlayerPawn)
        (CallFunction|ManagePilotExit)
        (else
          (Utilities|IsValid PlayerPawn
            (:Is%20Valid
              (Utilities|Array|RemoveItem (Variables|Default|GetPassengerPawns) PlayerPawn)
              (CallFunction|ManagePassengerExit PlayerPawn))))))))

(event Custom|StopRotorHitCamera
  ;[1 1 1] Rotor Hit Haptics
  (Utilities|IsValid (Variables|CameraShake|GetRotorImpactCameraShake)
    (:Is%20Valid
      (CameraShakes|StopCameraShake (Variables|/Script/Engine.PlayerController|GetPlayerCameraManager (Variables|Default|GetPilotController)) (Variables|CameraShake|GetRotorImpactCameraShake) true)
      (Variables|CameraShake|SetRotorImpactCameraShake 0))))

(event EventOnRotorDamageDealtInner
  (params (Impact :type Hit%20Result%20Structure%20%28by%20ref%29))
  (bind (_bblockinghit _binitialoverlap _time _distance _location _impactpoint _normal _impactnormal _physmat _hitactor) (Collision|BreakHitResult Impact))
  (bind _returnvalue (Vehicle|GetDriver))
  ;[1 1 1] Rotor Impact - Inner
  (bind _asfort_aipawn (Utilities|Casting|CastToFortAIPawn _hitactor)
    (:then
      (if (Networking|IsServer)
        (CallFunction|RotorDamagePawns _asfort_aipawn)
        (else
          _)))
    (:CastFailed
      _))
  (Utilities|IsValid _hitactor
    (:Is%20Valid
      (Utilities|IsValid _returnvalue
        (:Is%20Valid
          (GameplayEffects|ApplyGameplayEffectToTarget (Ability|GetAbilitySystemComponent _returnvalue) (Variables|GameplayEffects|GetGE_RotorHit_DamageBuilding_Inner) (Ability|GetAbilitySystemComponent _hitactor) 0.0 (GameplayEffects|MakeEffectContext (Ability|GetAbilitySystemComponent _returnvalue)))
          (CallFunction|RotorHitHaptics)
          ;[1 1 1] Don't damage if we hit small props
          (bind _asbuilding_prop (Utilities|Casting|CastToBuildingProp _hitactor)
            (:then
              (if (CallFunction|ShouldDamageBuilding _asbuilding_prop 125.0)
                _
                (else
                  _)))
            (:CastFailed
              _))
          (Utilities|FlowControl|DoOnce false
            (:Completed
              (Utilities|Time|SetTimerbyEvent (delegate Custom|ResetRotorDamage) 0.75)
              (GameplayEffects|ApplyGameplayEffectToSelf (Vehicle|GetVehicleAbilitySystemComponent) (Variables|GameplayEffects|GetGE_RotorHit_DamageSelf_Inner) 0.0 (GameplayEffects|MakeEffectContext (Vehicle|GetVehicleAbilitySystemComponent))))))))))

(event Custom|RotorHitHaptics
  (bind _pilot (Variables|Default|GetPilot))
  (bind _pilotcontroller (Variables|Default|GetPilotController))
  ;[1 1 1] Rotor Hit Haptics
  (Utilities|IsValid _pilot
    (:Is%20Valid
      (Utilities|IsValid (Variables|Default|GetPilotController)
        (:Is%20Valid
          (Fortnite|PlayLocalForceFeedbackatLocation "/Game/Weapons/ForceFeedback/FF_Generic_Impact_Near.FF_Generic_Impact_Near" (Transformation|GetActorLocation _pilot) 11.0 "RotorHitFF")
          (Utilities|IsValid (Variables|CameraShake|GetRotorImpactCameraShake)
            (:Is%20Valid
              _)
            (:Is%20Not%20Valid
              (bind _returnvalue (CameraShakes|StartCameraShake :self (Variables|/Script/Engine.PlayerController|GetPlayerCameraManager (Variables|Default|GetPilotController)) :ShakeClass "/Hoagie/Hoagie_RotorHit_CameraShake.Hoagie_RotorHit_CameraShake_C" :Scale 1.0 :PlaySpace "CameraLocal"))
              (Variables|CameraShake|SetRotorImpactCameraShake (CameraShakes|ConvLegacyCameraShake _returnvalue))))
          (Utilities|FlowControl|RetriggerableDelay 0.3) ; Stop Camera shake if we don't hit anything
          (Utilities|IsValid _pilotcontroller
            (:Is%20Valid
              (CameraShakes|StopCameraShake (Variables|/Script/Engine.PlayerController|GetPlayerCameraManager _pilotcontroller) (Variables|CameraShake|GetRotorImpactCameraShake) true)
              (Variables|CameraShake|SetRotorImpactCameraShake 0))))))))

(event Custom|ResetRotorDamage
  ;[1 1 1] Apply Damage to Vehicle from Rotor Hit
  (Utilities|FlowControl|DoOnce false
    (:Completed
      (Utilities|Time|SetTimerbyEvent (delegate Custom|ResetRotorDamage) 0.75)
      (GameplayEffects|ApplyGameplayEffectToSelf (Vehicle|GetVehicleAbilitySystemComponent) (Variables|GameplayEffects|GetGE_RotorHit_DamageSelf_Inner) 0.0 (GameplayEffects|MakeEffectContext (Vehicle|GetVehicleAbilitySystemComponent))))))

(event Custom|AltimeterTrace
  ;[1 1 1] AlimeterTrace
  (Variables|Atimeter|SetCurrentAltitude (Math|Float|Round (/ (Vehicle|GetDistancetoGround) 1e+02)))
  ;[1 1 1] Set Camera Location Shake
  (Utilities|IsValid (Variables|CameraShake|GetFlightCameraShake)
    (:Is%20Valid
      (Variables|/Script/EngineCameras.LegacyCameraShake|SetShakeScale (Math|Float|MapRangeClamped (Math|Conversions|ToFloat(Integer) (Variables|Atimeter|GetCurrentAltitude)) 3.0 8.0 0.01 1.0) (Variables|CameraShake|GetFlightCameraShake)))))

(event Custom|EventOnHealthChanged
  ;[0.1559 0.1875 0.04476] Event On Health Changed
  (bind _output_get (Variables|Default|SetDamageValue (Math|Float|Lerp 1.0 0.0 (/ (Class|FortVehicleInterface|GetHealth) (Class|FortVehicleInterface|GetMaxHealth)))))
  (CallFunction|UpdateDamageState _output_get))

(event EventTick
  (params (DeltaSeconds :type Float%20%28single-precision%29))
  ;[0.3438 1 0.3438] TICK
  (bind _audiocontroller (Utilities|GetAudioController)
    (:Success
      (Vehicle|Update _audiocontroller))))

(event Custom|RotorSkidRumble
  ;[1 1 1] Rumble while skidding
  (CallFunction|StopSkidRumble))

(event Custom|UpdateDamageState
  (params (Damage :type Float%20%28double-precision%29))
  ;[1 1 1] Update Mat and VFX based on damage
  (Vehicle|UpdateDamageStateNative Damage)
  (Networking|SwitchHasAuthority
    (:Remote
      (bind _audiocontroller (Utilities|GetAudioController)
        (:Success
          (Class|BPHoagieAudioController|UpdateDamageState _audiocontroller Damage))
        (:Failed
          _))
      (Niagara|SetNiagaraVariable(Bool) (Variables|Default|GetDamageFX) "User.isCritical" (>= Damage 0.95)))))

(event Custom|StopSkidRumble
  ;[1 1 1] Rumble while skidding
  (Utilities|Time|ClearandInvalidateTimerbyHandle (Variables|Default|GetRotorSkidTimer))
  (Game|Feedback|PlayDynamicForceFeedback (Variables|Default|GetPilotController) (Variables|Default|GetSkidIntensity) -1 true false true false))

(event EventOnCollisionHitEffects
  (params (HitLocation :type Vector%20%28by%20ref%29 :default "0, 0, 0") (HitNormalImpulse :type Vector%20%28by%20ref%29 :default "0, 0, 0") (HitFrictionImpulse :type Vector%20%28by%20ref%29 :default "0, 0, 0") (HitNormal :type Vector%20%28by%20ref%29 :default "0, 0, 0") (HitActor :type Actor%20Object%20Reference) (HitSurfaceType :type EPhysicalSurface%20Enum :default "SurfaceType_Default"))
  (bind _returnvalue (Math|Vector|VectorLength HitNormalImpulse))
  ;[0.969 0.3804 1] Audio/FX collision, used for skid fx when the heli slides on the ground
  (bind _audiocontroller (Utilities|GetAudioController)
    (:Success
      (Class|BPHoagieAudioController|OnNotifyCollisionHit _audiocontroller HitLocation HitNormalImpulse))
    (:Failed
      _))
  (if (Math|Boolean|ANDBoolean (> _returnvalue 5e+03) (>= (- (Utilities|Time|GetGameTimeinSeconds) (Variables|Default|GetLastCollisionTime)) (Variables|Default|GetMinTimeBetweenCollision)) (Vehicle|IsScrapingBottom))
    (Variables|Default|SetLastCollisionTime (Utilities|Time|GetGameTimeinSeconds))
    (CallFunction|SpawnImpactEffectSkid HitLocation HitNormal _returnvalue HitSurfaceType)
    ;[1 1 1] Rumble while skidding
    (Utilities|IsValid (Variables|Default|GetPilotController)
      (:Is%20Valid
        (bind _returnvalue_2 (Utilities|Time|SetTimerbyEvent (delegate Custom|RotorSkidRumble) 0.15))
        (Variables|Default|SetRotorSkidTimer _returnvalue_2))
      (:Is%20Not%20Valid
        (Utilities|Time|ClearandInvalidateTimerbyHandle (Variables|Default|GetRotorSkidTimer))))
    (Game|Feedback|PlayDynamicForceFeedback (Variables|Default|GetPilotController) (Variables|Default|GetSkidIntensity) -1 true false true false)))

(event Custom|StartPassiveRumble
  ;[1 1 1] Rumble - Ascend / Descend
  (Game|Feedback|PlayDynamicForceFeedback (Variables|Default|GetPilotController) (Variables|Default|GetPassiveRumbleIntensity) -1 (Variables|Default|GetRumbleLarge) (Variables|Default|GetRumbeSmall) (Variables|Default|GetRumbleLarge) (Variables|Default|GetRumbeSmall)))

(event Custom|UpdatePassiveRumble
  ;[1 1 1] Rumble - Ascend / Descend
  (Game|Feedback|PlayDynamicForceFeedback (Variables|Default|GetPilotController) (Variables|Default|GetPassiveRumbleIntensity) -1 (Variables|Default|GetRumbleLarge) (Variables|Default|GetRumbeSmall) (Variables|Default|GetRumbleLarge) (Variables|Default|GetRumbeSmall)))

(event EventOnDamagePlayEffects
  (params (Damage :type Float%20%28single-precision%29) (DamageTags :type Gameplay%20Tag%20Container%20Structure%20%28by%20ref%29) (Momentum :type Vector :default "0, 0, 0") (HitInfo :type Hit%20Result%20Structure%20%28by%20ref%29) (InstigatedBy :type Fort%20Pawn%20Object%20Reference) (DamageCauser :type Actor%20Object%20Reference) (EffectContext :type Gameplay%20Effect%20Context%20Handle%20Structure))
  (bind (_bblockinghit _binitialoverlap _time _distance _location _impactpoint _normal) (Collision|BreakHitResult HitInfo))
  ;[1 1 1] Adding Hit Impact Effects
  (if (> Damage 21.0)
    (Niagara|SpawnSystematLocation (Variables|FX|GetFX%20Large%20Impact) _location (Math|Conversions|RotationFromXVector _normal) (Math|Conversions|ToVector(Float) (Math|Float|SelectFloat 1.5 1.0 (> Damage 5e+01))) true true "None" false)))

(event Custom|StopPassiveRumble
  ;[1 1 1] Rumble - Ascend / Descend
  (Game|Feedback|PlayDynamicForceFeedback (Variables|Default|GetPilotController) (Variables|Default|GetPassiveRumbleIntensity) -1 (Variables|Default|GetRumbleLarge) (Variables|Default|GetRumbeSmall) (Variables|Default|GetRumbleLarge) (Variables|Default|GetRumbeSmall)))

(event EventOnBoostReady
  ;[1 1 1] BOOST
  (Utilities|IsValid (Variables|Default|GetPilot)
    (:Is%20Valid
      (bind _audiocontroller (Utilities|GetAudioController)
        (:Success
          (Class|BPHoagieAudioController|OnNotifyBoostReady _audiocontroller))
        (:Failed
          _))
      (if (not (Variables|Boost|GetIsBoosting))
        (CallFunction|BoosterFadesIn)))))

(event EventOnDamaged
  (params (Damage :type Float%20%28single-precision%29) (DamageTags :type Gameplay%20Tag%20Container%20Structure%20%28by%20ref%29) (EffectContext :type Gameplay%20Effect%20Context%20Handle%20Structure%20%28by%20ref%29) (EventInstigator :type Controller%20Object%20Reference) (DamageCauser :type Actor%20Object%20Reference))
  (bind (_returnvalue_bblockinghit _returnvalue_binitialoverlap _returnvalue_time _returnvalue_distance _returnvalue_location _returnvalue_impactpoint _returnvalue_normal _returnvalue_impactnormal _returnvalue_physmat _returnvalue_hitactor _returnvalue_hitcomponent) (Collision|BreakHitResult (Ability|EffectContext|GetHitResult EffectContext)))
  (bind _returnvalue (Pawn|GetControlledPawn EventInstigator))
  ;[1 1 1] EVENT -> On Damaged: Hit with Pickaxe
  (Utilities|IsValid EventInstigator
    (:Is%20Valid
      (|CallParentFunction Damage DamageTags EffectContext EventInstigator DamageCauser)
      (if (Utilities|Equal(Object) _returnvalue (Variables|Default|GetPilot))
        (if (not (Vehicle|Destruction|IsCriticalHealth))
          (Variables|Default|SetLastDamageCauser 0))
        (else
          (Variables|Default|SetLastDamageCauser _returnvalue)))
      (CallFunction|CheckHitGlass (Utilities|GetDisplayName _returnvalue_hitcomponent))
      (if (GameplayTags|HasTag DamageTags "(TagName=\"Weapon.Melee.Impact.Pickaxe\")")
        (if (> 1.0 0.0)
          (Variables|Default|SetHitPickaxePawn DamageCauser)
          (CallFunction|HitwithPickaxe (Math|Vector|GetForwardVector (Pawn|GetControlRotation EventInstigator :!class /Script/Engine.Controller))))))))

(event EventOnBoostStarted
  (bind _pilotcontroller (Variables|Default|GetPilotController))
  (bind _pilotcontroller_2 (Variables|Default|GetPilotController))
  (bind _playercameramanager (Variables|/Script/Engine.PlayerController|GetPlayerCameraManager _pilotcontroller_2))
  (bind _projectile_class (Variables|Default|GetProjectile%20Class))
  ;[1 1 1] BOOST
  (Variables|Boost|SetIsBoosting true)
  (CallFunction|SpawnBooster)
  (Effects|Components|ParticleSystem|SetFloatParameter (Variables|Default|GetHoagie%20Idle%20FX) "User.BoostReady" 0.0)
  (CallFunction|FuelFadeOut)
  (Utilities|FlowControl|Sequence
    (:then_0
      ;;[0.7047 0.1898 1] Boost Effects (audio, rumble, camera shake)
      (Utilities|IsValid _pilotcontroller
        (:Is%20Valid
          (Game|Feedback|PlayDynamicForceFeedback _pilotcontroller (stale _intensity) -1 true true true true))))
    (:then_1
      (bind (_direction _intensity) (Timeline|Timeline_0 PlayFromStart)
        (:Update
          (Utilities|IsValid _pilotcontroller
            (:Is%20Valid
              _)))
        (:Finished
          (Utilities|IsValid _pilotcontroller
            (:Is%20Valid
              _)))))
    (:then_2
      (bind _audiocontroller (Utilities|GetAudioController)
        (:Success
          (Class|BPHoagieAudioController|OnNotifyBoostStart _audiocontroller))
        (:Failed
          _))
      (Utilities|IsValid _pilotcontroller_2
        (:Is%20Valid
          (Utilities|IsValid _playercameramanager
            (:Is%20Valid
              (bind _returnvalue (CameraShakes|StartCameraShake :self _playercameramanager :ShakeClass (Variables|CameraShake|GetCameraShakeBoost) :Scale 1.0 :PlaySpace "CameraLocal"))
              (Variables|CameraShake|SetBoostCameraShake (CameraShakes|ConvLegacyCameraShake _returnvalue)))))))
    (:then_3
      (Utilities|FlowControl|RetriggerableDelay 0.25)
      (CallFunction|BoosterFadesOut))
    (:then_4
      (Projectile|SpawnProjectile :ProjectileClass _projectile_class :RequestedBy (Vehicle|GetDriver) :SpawnLocation (+ (Transformation|GetActorLocation (Vehicle|GetDriver)) "-200.000000,0.000000,0.000000") :SpawnRotation (Transformation|GetActorRotation (Vehicle|GetDriver)) :SpawnDirection (Transformation|GetActorRotation (Vehicle|GetDriver)) :bUseDefaultPhysics false :InitialSpeed 5e+01 :GravityScale 2.0 :ChargePercent 0.0)
      (Utilities|FlowControl|DelayUntilNextTick)
      (Projectile|SpawnProjectile :ProjectileClass _projectile_class :RequestedBy (Vehicle|GetDriver) :SpawnLocation (Transformation|GetActorLocation (Vehicle|GetDriver)) :SpawnRotation (Transformation|GetActorRotation (Vehicle|GetDriver)) :SpawnDirection (Transformation|GetActorRotation (Vehicle|GetDriver)) :bUseDefaultPhysics false :InitialSpeed 5e+01 :GravityScale 2.0 :ChargePercent 0.0))))

(event OnEnteredWaterDelegate(PontoonsComponent)
  (params (Pontoon :type Spherical%20Pontoon%20Structure%20%28by%20ref%29))
  (bind (_centersocket _relativelocation _radius _bfxenabled _localforce _centerlocation _socketrotation _offset _waterheight _waterdepth _immersiondepth _waterplanelocation _waterplanenormal) (Utilities|Struct|BreakSphericalPontoon Pontoon))
  (bind _returnvalue (Effects|Components|ParticleSystem|SpawnFXAtLocation :FXSystemAsset "/Hoagie/FX/NS_Hoagie_WaterSplashDown.NS_Hoagie_WaterSplashDown" :Location _centerlocation :Scale "1.000000,1.000000,1.000000" :bAutoDestroy true :PoolingMethod "AutoRelease" :bAutoActivate true))
  (Effects|Components|ParticleSystem|SetFloatParameter _returnvalue "User.WaterDepth" _waterdepth)
  (Effects|Components|ParticleSystem|SetVectorParameter _returnvalue "User.WaterPlaneLocation" _waterplanelocation)
  (Effects|Components|ParticleSystem|SetVectorParameter _returnvalue "User.WaterPlaneNormal" _waterplanenormal)
  (bind _audiocontroller (Utilities|GetAudioController)
    (:Success
      (Class|BPHoagieAudioController|OnEnteredWater _audiocontroller _centerlocation))))

(event Custom|StopBoostRumble
  (bind _pilotcontroller (Variables|Default|GetPilotController))
  ;[1 1 1] BOOST
  ;;[0.7047 0.1898 1] Boost Effects (audio, rumble, camera shake)
  (Utilities|IsValid _pilotcontroller
    (:Is%20Valid
      (Game|Feedback|PlayDynamicForceFeedback _pilotcontroller (stale EventOnBoostStarted _intensity) -1 true true true true))))

(event Custom|CheckHitGlass
  (params (ComponentName :type String))
  ;[1 1 1] EVENT -> On Damaged: Hit with Pickaxe
  ;;[1 1 1] If glass was hit, destroy it
  (Utilities|CheckGlassHit ComponentName "Front"
    (:true
      (Variables|Glass|SetGlassFrontDestroyed true))
    (:false
      (Utilities|CheckGlassHit ComponentName "Right"
        (:true
          (Variables|Glass|SetGlassRightDestroyed true))
        (:false
          (Utilities|CheckGlassHit ComponentName "Left"
            (:true
              (Variables|Glass|SetGlassLeftDestroyed true))))))))

(event EventOnBoostFinished
  ;[1 1 1] BOOST
  (Variables|Boost|SetIsBoosting false)
  (CallFunction|DespawnBooster)
  (bind _audiocontroller (Utilities|GetAudioController)
    (:Success
      (Class|BPHoagieAudioController|OnNotifyBoostEnd _audiocontroller))))

(event Custom|Booster%20Fades%20In
  ;[1 1 1] BOOST
  (bind (_direction _fade_lerp) (Timeline|BoosterFades PlayFromStart)
    (:Update
      (Niagara|SetNiagaraVariable(Float) (Variables|Default|GetHoagie%20Idle%20FX) "User.BoostReady" _fade_lerp))))

(event Custom|Booster%20Fades%20Out
  ;[1 1 1] BOOST
  (Timeline|BoosterFades ReverseFromEnd))

(event Custom|Fuel%20Fade%20Out
  ;[1 1 1] BOOST
  (bind (_direction _newtrack_0) (Timeline|Fuel%20Gauge%20Off PlayFromStart)
    (:Update
      (Rendering|Material|SetScalarParameterValueonMaterials (Variables|Vehicle|GetSkeletalMesh) "FuelLevel" _newtrack_0))
    (:Finished
      (Utilities|FlowControl|RetriggerableDelay 0.5)
      (bind (_direction_2 _newtrack_0_2) (Timeline|FuelGaugeOn PlayFromStart)
        (:Update
          (Rendering|Material|SetScalarParameterValueonMaterials (Variables|Vehicle|GetSkeletalMesh) "FuelLevel" _newtrack_0_2))))))

(event EventShowCooldownCue
  (params (Pawn :type Fort%20Player%20Pawn%20Object%20Reference) (Duration :type Float%20%28single-precision%29))
  ;[1 1 1] BOOST
  ;;[1 1 1] Boost Cooldown Cue
  (Ability|SendGameplayEventtoActor Pawn "(TagName=\"GameplayEffect.Vehicle.Boost\")" (Utilities|Struct|MakeGameplayEventData :ContextHandle "()" :InstigatorTags "(GameplayTags=)" :TargetTags "(GameplayTags=)" :EventMagnitude Duration :TargetData "()")))

(event Custom|HitWithPickaxe
  (params (HitDirection :type Vector))
  (bind _returnvalue (Transformation|GetActorForwardVector))
  (bind (_pickaxeimpulse_x _pickaxeimpulse_y _pickaxeimpulse_z) (Math|Vector|BreakVector (Variables|GetPickaxeImpulse_X)))
  ;[1 1 1] Hit with Pickaxe
  (if (> (Variables|Default|GetShouldPickaxeImpulse) 0.0)
    (if (Vehicle|VehicleIsinAir)
      _
      (elif (and (<= (Math|Vector|DotProduct _returnvalue HitDirection) 0.0) (>= (Vehicle|GetPreviousForwardSpeedKmh) 2e+01))
        (Variables|Default|SetPickaxeImpulse (Math|Vector|vector*float (Math|Vector|MakeVector (.x _returnvalue) (.y _returnvalue) 1.15) (Variables|Default|GetPickaxeImpulseXYMulti)))
        (if (not (CallFunction|IsinWaterBody))
          (Variables|Default|SetPickaxeImpulse (* (Variables|Default|GetPickAxeImpulse) "0.500000,0.500000,1.000000")))
        (Audio|BroadcastAbilityImpactatLocation 0 self (Math|Vector|MakeVector _pickaxeimpulse_x _pickaxeimpulse_y (Math|Float|Max(Float) _pickaxeimpulse_z (Variables|Default|GetPickaxeImpulseMinZ))) 777.0 "None" false)
        (else
          (Variables|Default|SetPickaxeImpulse (Math|Vector|vector*float (Math|Vector|MakeVector (.x HitDirection) (.y HitDirection) (select (>= (.z HitDirection) 0.0) (Math|Float|Clamp(Float) (.z HitDirection) 0.4 0.8) 0.3)) (Variables|Default|GetPickaxeImpulseXYMulti))))))))

(event EventOnRotorDamagePlayer
  (params (Impact :type Hit%20Result%20Structure%20%28by%20ref%29))
  (bind (_bblockinghit _binitialoverlap _time _distance _location _impactpoint _normal _impactnormal _physmat _hitactor) (Collision|BreakHitResult Impact))
  ;[1 1 1] Bounce Off Rotors
  (bind _asfort_player_pawn (Utilities|Casting|CastToFortPlayerPawn _hitactor))
  (if (not (GameplayTags|HasMatchingGameplayTag _asfort_player_pawn (Variables|Default|GetRecentlyExitTag)))
    (bind _output_get (Variables|Default|SetPlayerHitByRotors _asfort_player_pawn))
    ;;[0.4427 0.4427 0.4427] Check to see if hit Player is in a Hoagie
    (if (Utilities|IsA (Vehicle|GetVehicleActor _output_get :!class /Script/FortniteGame.FortPlayerPawn) "/Script/HoagieRuntime.FortHoagieVehicle")
      _
      (elif (Networking|IsServer)
        (CallFunction|RotorDamagePawns (Variables|Default|GetPlayerHitByRotors))
        (bind _asfort_player_pawn_athena (Utilities|Casting|CastToFortPlayerPawnAthena (Variables|Default|GetPlayerHitByRotors)))
        (Class|FortPlayerPawnAthena|SetFallInstigator :self _asfort_player_pawn_athena :NewFallInstigator (Game|GetInstigatorController (Variables|Default|GetPilot)))))))

(event Custom|RotorDamagePawns
  (params (TargetPawn :type Fort%20Pawn%20Object%20Reference))
  (bind (_returnvalue_x _returnvalue_y _returnvalue_z) (Math|Vector|BreakVector (* (Math|Vector|Normalize (- (Transformation|GetActorLocation TargetPawn) (Vehicle|GetRotorCenterPosition)) 0.1) (Variables|Default|GetRotorLaunchPlayerVelocity))))
  (bind _pilot (Variables|Default|GetPilot))
  ;[1 1 1] Bounce Off Rotors
  (Utilities|IsValid TargetPawn
    (:Is%20Valid
      (Utilities|IsValid (Variables|Default|GetLastHitPlayer)
        (:Is%20Valid
          _)
        (:Is%20Not%20Valid
          (Character|LaunchCharacter TargetPawn (Math|Vector|MakeVector _returnvalue_x _returnvalue_y (* _returnvalue_z 1.2)) true true))) ; If a player was hit recently, don't keep launching, lets them fly!
      (if (not (Team|OnSameTeam TargetPawn _pilot))
        (Utilities|IsValid _pilot
          (:Is%20Valid
            (Fortnite|ApplyGameplayEffectsToActor TargetPawn (Utilities|Array|MakeArray "/Hoagie/GE_Hoagie_RotorPlayerDamage.GE_Hoagie_RotorPlayerDamage_C") (Transformation|GetActorLocation) false 0 (Ability|GetAbilitySystemComponent _pilot)))) ; Only Damage if there's a pilot
      ))))

(event EventOnStopTick
  ;[1 1 1] Event On Stop Tick
  (Utilities|FlowControl|Sequence
    (:then_0
      (bind _audiocontroller (Utilities|GetAudioController)
        (:Success
          (Class|BPHoagieAudioController|OnNotifyStopTick _audiocontroller))))
    (:then_1
      (Networking|SwitchHasAuthority
        (:Remote
          (Class|RotorWashComponent|StopWashForceImmediately (Variables|Vehicle|GetRotorWashComponent)))))))

(event EventUpdateHoagieAnimBP
  (bind _skeletalmesh (Variables|Vehicle|GetSkeletalMesh))
  ;[0.1978 0.4971 1] Event Update Hoagie Anim BP
  ; Called when Skeletal Mesh's Anim Instance changes so that this blueprint updates animations correctly
  (|CallParentFunction)
  (Utilities|IsValid _skeletalmesh
    (:Is%20Valid
      (bind _assk_hoagie_protomesh_anim_bp (Utilities|Casting|CastToSK_Hoagie_Protomesh_AnimBP (Components|SkeletalMesh|GetAnimInstance _skeletalmesh)))
      (Variables|Animation|SetHoagieAnimBP _assk_hoagie_protomesh_anim_bp))))

(event OnComponentHit(SkeletalMesh)
  (params (HitComponent :type Primitive%20Component%20Object%20Reference) (OtherActor :type Actor%20Object%20Reference) (OtherComp :type Primitive%20Component%20Object%20Reference) (NormalImpulse :type Vector :default "0, 0, 0") (Hit :type Hit%20Result%20Structure%20%28by%20ref%29))
  (bind _returnvalue (Vehicle|GetPreviousForwardSpeedKmh))
  (bind (_returnvalue_x _returnvalue_y _returnvalue_z) (Math|Vector|BreakVector (* (Transformation|GetActorForwardVector) (* _returnvalue (Variables|Default|GetSpeedImpulsePlayerMulti)))))
  (bind _returnvalue_2 (Vehicle|GetPreviousForwardSpeedKmh))
  (bind (_returnvalue_x_2 _returnvalue_y_2 _returnvalue_z_2) (Math|Vector|BreakVector (* (Transformation|GetActorForwardVector) (* _returnvalue_2 (Variables|Default|GetSpeedImpulsePlayerMulti)))))
  (bind _minspeedtolaunchplayer (Variables|Default|GetMinSpeedToLaunchPlayer))
  ;[1 1 1] EVENT -> On Hit (Skel Mesh)
  (bind _asbuilding_actor (Utilities|Casting|CastToBuildingActor OtherActor)
    (:then
      (Utilities|FlowControl|Sequence
        (:then_0
          ;[1 0.6042 0.6042] Blow up explosive props if we hit them
          (bind _asathena_prop_burst_reaction_exploding (Utilities|Casting|CastToAthena_Prop_BurstReaction_Exploding _asbuilding_actor))
          (Actor|DestroyActor _asathena_prop_burst_reaction_exploding))
        (:then_1
          ;[1 1 1] Destroy small props hit
          (bind _asbuilding_prop (Utilities|Casting|CastToBuildingProp _asbuilding_actor))
          (if (CallFunction|ShouldDamageBuilding _asbuilding_prop 125.0)
            (GameplayEffects|ApplyGameplayEffectToTarget :self (Ability|GetAbilitySystemComponent self) :GameplayEffectClass (Variables|GameplayEffects|GetGE_Hit_DestroyProp) :Target (Ability|GetAbilitySystemComponent _asbuilding_prop) :Level 1.0)))))
    (:CastFailed
      ;[1 1 1] Launch Player
      (bind _asfort_player_pawn_athena (Utilities|Casting|CastToFortPlayerPawnAthena OtherActor)
        (:then
          (if (GameplayTags|HasMatchingGameplayTag _asfort_player_pawn_athena (Variables|Default|GetRecentlyExitTag))
            _
            (elif (Parachute|IsSkydiving _asfort_player_pawn_athena)
              _
              (elif (Utilities|Array|ContainsItem (Variables|Default|GetPassengerPawns) _asfort_player_pawn_athena)
                _
                (elif (Utilities|Equal(Object) _asfort_player_pawn_athena (Variables|Default|GetLastHitPlayer))
                  _
                  (else
                    (Variables|Default|SetLastHitPlayer _asfort_player_pawn_athena)
                    (Utilities|FlowControl|Sequence
                      (:then_0
                        (if (Networking|IsServer)
                          (if (>= _returnvalue_2 _minspeedtolaunchplayer)
                            (Character|LaunchCharacter _asfort_player_pawn_athena (Math|Vector|MakeVector _returnvalue_x_2 _returnvalue_y_2 (Math|Float|Max(Float) _returnvalue_z_2 1.5e+03)) true true) ; Launch player based on cart velocity
                            (for _array_element (Vehicle|GetAllPassengers)
                              (GameplayCue|ExecuteGameplayCue%28Burst%29Local _array_element (Variables|GameplayCues|GetGC_HitPlayer) (Variables|GameplayCues|GetGC_ParamsEmpty))
                              (if (Vehicle|HasDriver)
                                (Class|FortPlayerPawnAthena|SetFallInstigator :self _asfort_player_pawn_athena :NewFallInstigator (Game|GetInstigatorController (Variables|Default|GetPilot)))
                                (else
                                  _))))))
                      (:then_1
                        ;; ResetLastHitPlayer
                        (Utilities|FlowControl|Delay 0.4)
                        (Variables|Default|SetLastHitPlayer 0)
                        (Variables|Default|SetHitPickaxePawn 0)))))))))
        (:CastFailed
          (bind _asfort_aipawn (Utilities|Casting|CastToFortAIPawn OtherActor))
          (Components|Movement|StopMovementImmediately (Variables|/Script/Engine.Character|GetCharacterMovement _asfort_aipawn))
          (if (>= _returnvalue _minspeedtolaunchplayer)
            (Character|LaunchCharacter _asfort_aipawn (Math|Vector|MakeVector _returnvalue_x _returnvalue_y (Math|Float|Max(Float) _returnvalue_z 1e+03)) true true) ; Launch player based on cart velocity
            (Networking|SwitchHasAuthority
              (:Remote
                (for _array_element_2 (Vehicle|GetAllPassengers)
                  (GameplayCue|ExecuteGameplayCue%28Burst%29Local _array_element_2 (Variables|GameplayCues|GetGC_HitFiend) (Variables|GameplayCues|GetGC_ParamsEmpty)))))))))))

(event Custom|ExplodeCosmetic
  ;[0.02031 0.1276 0.1302] ExplodeCosmetics: Handles logic related to the cosmetics of the vehicle explosion.
  (Networking|CanExecuteCosmeticEvents
    (:True
      (Niagara|SpawnSystematLocation (Variables|FX|GetFX_HoagieExplode) (Transformation|GetWorldLocation (Variables|Vehicle|GetSkeletalMesh)) (Transformation|GetActorRotation))
      (bind _audiocontroller (Utilities|GetAudioController)
        (:Success
          (Class|BPHoagieAudioController|OnNotifyDestroyed _audiocontroller))
        (:Failed
          _))
      (Rendering|SetVisibility (Variables|Default|GetHoagie%20Idle%20FX) false false)
      (CallFunction|SpawnWreckFX)
      (Fortnite|ShowSoundIndicatorLocalforTeam :Location (Transformation|GetActorLocation self) :MaxAudibleDistance 4e+03 :IndicatorType "Gunshot" :Team (Team|GetActorTeam self) :Affiliations (Utilities|Array|MakeArray "Hostile" "Neutral" "Friendly") :Instigator self :Tint "(R=0.000000,G=0.000000,B=0.000000,A=0.000000)" :AttenuationMaxRangeMultiplier 1.0 :AttenuationIntensityMultiplier 1.0)
      (GameplayCue|ExecuteGameplayCue%28Burst%29Local self (Variables|GameplayCues|GetGC_ExplodeShake) (Ability|GameplayCue|MakeGameplayCueParameters :NormalizedMagnitude 1.0 :RawMagnitude 1.0 :Location "0,0,0" :Normal "0,0,0" :Instigator self :EffectCauser self :GameplayEffectLevel 1 :AbilityLevel 1 :bReplicateLocationWhenUsingMinimalRepProxy false))
      (bind _returnvalue (Audio|SpawnSoundAttached :Sound (Variables|Default|GetDamageEngineFireSFX) :AttachToComponent (Variables|Vehicle|GetSkeletalMesh) :AttachPointName "FX_Wreck" :LocationType "KeepRelativeOffset" :bStopWhenAttachedToDestroyed true :VolumeMultiplier 1.0 :PitchMultiplier 1.0 :StartTime 0.0 :bAutoDestroy true))
      (Variables|Default|SetDamageEngineFireSFXComp _returnvalue)
      ;;[0.005199 0.1667 0.165] Apply destroyed material to the body material elements.
      (for _array_element (Variables|Default|GetBodyMaterialIndexes)
        (Rendering|Material|SetMaterial (Variables|Vehicle|GetSkeletalMesh) _array_element (Variables|Default|GetDestroyedMaterial))))))

(event Custom|FadeVehicleHusk
  ;[0.1436 0.1458 0.03842] FadeVehicleHusk: Fades out the vehicle husk through material param change.
  (Networking|SwitchHasAuthority
    (:Authority
      (Components|Timeline|SetPlayRate (Variables|HoagieVehicle_X|GetTimeline_FadeHusk) (Math|Float|SafeDivide 1.0 (Variables|Default|GetFadeDuration)))
      (bind (_direction _alpha) (Timeline|Timeline_FadeHusk Play)
        (:Update
          (Rendering|Material|SetScalarParameterValueonMaterials (Variables|Vehicle|GetSkeletalMesh) "Fade" _alpha))
        (:CleanFX
          (CallFunction|CleanUpFXs)
          ;;[0.03905 0.2083 0.2063] Clean up the native refernces so they are not activated again.
          (Variables|Vehicle|SetDamageFX_Native 0)
          (Variables|Vehicle|SetHoagieIdleFX_Native 0)
          (Variables|Vehicle|SetRotorWashFX_Native 0))) ; Replicated timeline.
    )))

(event Custom|StartCurieFire
  ;[0.01934 0.1094 0.1072] StartCurieFire: When the health of the vehicle reaches critical amount (here 0) and it catches fire, start a curie fire around the actor.
  (Utilities|FlowControl|DoOnce false
    (:Completed
      (Networking|SwitchHasAuthority
        (:Authority
          (if (Variables|Default|GetbValetCurieEnabled)
            (Curie|OnCurieActivated
              (:OnCurieActive
                (Curie|EnableElementAttachmentonActor self "(TagName=\"Curie.Element.Fire\")")
                (Curie|ExecuteFireApplication self self self (Variables|Default|GetDestructionCurieMagnitude) (Collision|MakeHitResult false false 0.0 0.0 (Transformation|GetActorLocation) (Transformation|GetActorLocation) "0,0,1" "0,0,1" 0 self (Variables|Vehicle|GetSkeletalMesh)) -1 (Variables|Default|GetDestructionCurieLandscapeIgnitionRadius))))))))) ; Do Once, since this event is invoked \nOnExplosion as well as when the vehicle catches fire at 0 \nhealth but we want to ever only start a fire related to this actor once \nbased on whatever happens first, the explosion or fire at 0 health.
)

(event Custom|SpawnWreckFX
  ;[1 0.1458 0.1678] SpawnWreckFX: Spawns a wreck FX (looks like fire) and is attached to the vehicle.
  (bind _wreckfxniagara (Variables|Default|GetWreckFXNiagara)
    (:then
      (bind _returnvalue (Niagara|SpawnSystemAttached _wreckfxniagara (Variables|Vehicle|GetSkeletalMesh) "FX_Wreck" "0, 0, 0" "0, 0, 0" "SnapToTarget" true true "AutoRelease"))
      (bind _output_get (Variables|Default|SetWreckFX _returnvalue))
      (Utilities|IsValid _output_get
        (:Is%20Valid
          (Niagara|SetNiagaraVariable(Float) _returnvalue "Duration" (CallFunction|GetMaxZeroHealthTime)))))
    (:else)))

(event Curie|EventOnCurieContainerAcquired
  (params (CurieContainerHandle :type Curie%20Container%20Handle%20Structure))
  ;[0.04101 0.1392 0.1406] Event OnCurieContainerAcquired: IF the actor is not on fire when the curie container is acquired, disable the element attachment on the actor.
  ;
  ; Setup similar to valet vehicles.
  (if (not (CallFunction|IsonFire))
    (Curie|DisableElementAttachmentonActor self "(TagName=\"Curie.Element.Fire\")")))

(event EventEndPlay
  (params (EndPlayReason :type EEndPlayReason%20Enum :default "Destroyed"))
  ;[0.1667 0.1137 0.02079] EndPlay: Clean up any pending FXs.
  (|CallParentFunction EndPlayReason)
  (Networking|CanExecuteCosmeticEvents
    (:True
      (CallFunction|CleanUpFXs)
      (bind _fn_radial_force (Variables|Default|GetFN%20Radial%20Force)
        (:then
          (Actor|DestroyActor _fn_radial_force))
        (:else)))
    (:False
      (Utilities|Time|ClearandInvalidateTimerbyHandle (Variables|Default|GetFadeVehicleHuskTimerHandle))
      (Utilities|Time|ClearandInvalidateTimerbyHandle (Variables|Default|GetDestroyVehicleTimerHandle)))))

(event EventOnImpactWhileCritical
  ;[0.01855 0.09896 0.09797] OnImpactWhileCritical: Invoked when the vehicle crashes after reaching zero fuel. At this time, set the health of the vehicle to 0.
  ; Otherwise the vehicle ends up having damagable health values in its husk state which is incorrect.
  (Networking|SwitchHasAuthority
    (:Authority
      (Class|FortVehicleInterface|SetHealthPercent 0.0))))

(event Custom|DestroyHoagie
  ;[0.2708 0.07916 0.2153] DestroyHoagie: Destroy the vehicle.
  ; Called from a timer that started on explosion with a duration of MaxZeroHealthTime.
  (Networking|SwitchHasAuthority
    (:Authority
      (Vehicle|DestroyVehicle))))

(event Custom|HandleOnExplode
  (params (LastDamageInstigator :type Controller%20Object%20Reference) (Vehicle :type Actor%20Object%20Reference))
  (bind _returnvalue (CallFunction|GetMaxZeroHealthTime))
  (bind _fadeduration (Variables|Default|GetFadeDuration))
  (bind _returnvalue_2 (- _returnvalue _fadeduration))
  ;[0.2344 0.121 0.02911] HandleOnExplode: Callback for when the vehicle explodes, runs on boths server and client.
  ; On explosion, damage nearby actors, start a curie fire, and start a timer at the end of which the vehicle is cleaned up.
  (Networking|SwitchHasAuthority
    (:Authority
      (Variables|Default|SetIsCritical false)
      (Variables|Default|SetIsRotorWashActive false)
      (CallFunction|DestroyAllDamageableParts) ; This destroys any windows that are still intact.
      (Default|DoExplosionDamage LastDamageInstigator)
      (bind _returnvalue_3 (Utilities|Time|SetTimerbyEvent (EventDispatchers|CreateEvent :!event Custom|DestroyHoagie) (CallFunction|GetMaxZeroHealthTime))) ; Start a timer at the end of which the \nvehicle actor itself is cleaned up.
      (Variables|Default|SetDestroyVehicleTimerHandle _returnvalue_3)
      (if (> _returnvalue_2 0)
        (bind _returnvalue_4 (Utilities|Time|SetTimerbyEvent (EventDispatchers|CreateEvent :!event Custom|FadeVehicleHusk) _returnvalue_2)) ; Start a timer after which the fade \nexploded vehicle husk fades out.
        (Variables|Default|SetFadeVehicleHuskTimerHandle _returnvalue_4)
        (else
          (CallFunction|FadeVehicleHusk)))
      (CallFunction|StartCurieFire))
    (:Remote
      (Class|FortKismetLibrary|TriggerRadialShockwave (Transformation|GetWorldLocation (Variables|Vehicle|GetSkeletalMesh)) (Variables|Default|GetExplosionRadialForceStrength) (Variables|Default|GetExplosionRadialForceImpactRadius) (Variables|Default|GetExplosionRadialForceTimeDecay) 1.5e+02 0.15)
      (CallFunction|ExplodeCosmetic)
      (CallFunction|SpawnDamageEngineFireFX))))

(event Custom|SpawnDamageEngineFireFX
  (bind _skeletalmesh (Variables|Vehicle|GetSkeletalMesh))
  ;[0.02104 0.1458 0.1386] SpawnDamageEngineFireFX: Start the engine fire VFX and SFX.
  (Utilities|FlowControl|DoOnce false
    (:Completed
      (Networking|CanExecuteCosmeticEvents
        (:True
          (bind _returnvalue (Niagara|SpawnSystemAttached (Variables|Default|GetDamageEngineFireFXTemplate) _skeletalmesh "FX_Wreck" "0, 0, 0" "0, 0, 0" "SnapToTarget" true true "AutoRelease"))
          (bind _output_get (Variables|Default|SetDamageEngineFireFX _returnvalue))
          (Niagara|SetNiagaraVariable(Float) _output_get "User.DamageScale" 1.0)
          (Niagara|SetNiagaraVariable(Float) _output_get "Vehicle_Speed" 0.0)
          (bind _returnvalue_2 (Audio|SpawnSoundAttached :Sound (Variables|Default|GetDamageEngineSFX) :AttachToComponent _skeletalmesh :AttachPointName "FX_Wreck" :LocationType "KeepRelativeOffset" :bStopWhenAttachedToDestroyed true :VolumeMultiplier 1.0 :PitchMultiplier 1.0 :StartTime 0.0 :bAutoDestroy true))
          (Variables|Default|SetDamageEngineSFXComp _returnvalue_2))))) ; Do Once, since this event is invoked \nOnExplosion as well as when the vehicle catches fire at 0 \nhealth but we want to only ever run this event once based \non whatever happens first, the explosion or the critical health.
)

(event EventOnFireStart
  ;[0.2396 0.02235 0.1768] OnFireStart: On the client, spawn the engine fire FXs.
  (Networking|CanExecuteCosmeticEvents
    (:True
      (CallFunction|SpawnDamageEngineFireFX))))

(event Vehicle|FX|EventOnSetVehicleFx
  (params (FxTag :type Gameplay%20Tag%20Structure) (FxSystem :type FXSystem%20Asset%20Object%20Reference))
  ;[1 1 1] OnSetVehicleFx
  (switch Utilities|FlowControl|Switch|SwitchonGameplayTag FxTag
    (:Athena.Vehicle.Hoagie.FX.BoostActive
      (Variables|FX|SetBoost%20Active%20Template FxSystem))
    (:Athena.Vehicle.Hoagie.FX.CritHit
      (Variables|FX|SetFX_CritHit FxSystem))
    (:Athena.Vehicle.Hoagie.FX.RotorScrapes
      (Variables|FX|SetFX_RotorScrapes FxSystem))
    (:Athena.Vehicle.Hoagie.FX.ScrapeLandingSkids
      (bind _asniagara_system (Utilities|Casting|CastToNiagaraSystem FxSystem))
      (Variables|FX|Impact|SetNS_Hoagie_Scrape_LandingSkids _asniagara_system))
    (:Athena.Vehicle.Hoagie.FX.LargeImpact
      (bind _asniagara_system_2 (Utilities|Casting|CastToNiagaraSystem FxSystem))
      (Variables|FX|SetFX%20Large%20Impact _asniagara_system_2))))

(fn ApplyBoostCooldown
  (params (PlayerPawn :type Fort%20Player%20Pawn%20Object%20Reference))
  (bind _returnvalue (Environment|ApplyGlobalEnvironmentGameplayEffecttoActor PlayerPawn (Variables|Default|GetGE_Boost))) ; Apply GE for mobile Boost radial
  (Variables|Default|SetCooldownCueEvent _returnvalue))

(fn CacheHotfixValues
  ;[0 0.1198 0.1067] Cache the hotfix values into variables.
  (Variables|Default|SetSpeedImpulsePlayerMulti (ScalableFloat|GetValueatLevel (Variables|HotfixRows|GetHotfix_SpeedImpulsePlayerMulti)))
  (Variables|Default|SetMinSpeedToLaunchPlayer (ScalableFloat|GetValueatLevel (Variables|HotfixRows|GetHotfix_MinSpeedToLaunchPlayer)))
  (Variables|Default|SetRotorLaunchPlayerVelocity (ScalableFloat|GetValueatLevel (Variables|HotfixRows|GetHotfix_RotorLaunchPlayerVelocity)))
  (Variables|Default|SetPassiveRumble (ScalableFloat|GetValueatLevel (Variables|HotfixRows|GetHotfix_PassiveRumble)))
  (Variables|Default|SetDecendRumble (ScalableFloat|GetValueatLevel (Variables|HotfixRows|GetHotfix_DecendRumble)))
  (Variables|Default|SetAscendRumble (ScalableFloat|GetValueatLevel (Variables|HotfixRows|GetHotfix_AscendRumble)))
  (Variables|Default|SetPickaxeImpulseMinZ (ScalableFloat|GetValueatLevel (Variables|HotfixRows|GetHotfix_PickaxeImpulseMinZ)))
  (Variables|Default|SetPickaxeImpulseXYMulti (ScalableFloat|GetValueatLevel (Variables|HotfixRows|GetHotfix_PickaxeImpulseXYMulti)))
  (Variables|Default|SetShouldPickaxeImpulse (ScalableFloat|GetValueatLevel (Variables|HotfixRows|GetHotfix_ShouldPickaxeImpulse)))
  (Variables|Default|SetExplodeRadius (ScalableFloat|GetValueatLevel (Variables|HotfixRows|GetHotfix_ExplodeRadius)))
  (Variables|Default|SetWeakpointDamageMultiplier (ScalableFloat|GetValueatLevel (Variables|HotfixRows|GetHotfix_WeakpointDamageMultiplier)))
  (Variables|Default|SetExitLaunchMultiplier (ScalableFloat|GetValueatLevel (Variables|HotfixRows|GetHotfix_ExitLaunchMultiplier)))
  (Variables|Default|SetbValetCurieEnabled (ScalableFloat|GetValueasBool (Variables|Default|GetHotfixValetCurieEnabled)))
  (Variables|Default|SetExplosionDamageVehicles (ScalableFloat|GetValueatLevel (Variables|Default|GetHotfixExplosionDamageVehicles)))
  (Variables|Default|SetExplosionDamagePlayers (ScalableFloat|GetValueatLevel (Variables|Default|GetHotfixExplosionDamagePlayers)))
  (Variables|Default|SetExplosionDamageBuildings (ScalableFloat|GetValueatLevel (Variables|Default|GetHotfixExplosionDamageBuildings))))

(fn CleanUpFXs
  ;[0.0183 0.1195 0.1458] Clean up the spawned FXs.
  (Audio|Components|Audio|FadeOut (Variables|Default|GetDamageEngineSFXComp) 0.2 0.0 "Linear" :!class /Script/Engine.AudioComponent)
  (Components|Activation|Deactivate (Variables|Default|GetWreckFX)))

(fn DespawnBooster
  (bind _boostactive (Variables|Default|GetBoostActive))
  (Utilities|IsValid _boostactive
    (:Is%20Valid
      (Components|Activation|Deactivate _boostactive)
      (Effects|Components|ParticleSystem|ReleasetoPool _boostactive))
    (:Is%20Not%20Valid
      _))
  (return))

(fn DoExplosionDamage
  (params (Instigator :type Controller%20Object%20Reference))
  (Networking|SwitchHasAuthority
    (:Authority
      (if (> (ScalableFloat|GetCachedScalableFloatValue (Variables|Default|GetHotfix_UseAsyncExplosion)) 0)
        (Ability|Advanced|GrantandActivateAbilityonSelfwithParams (Vehicle|GetVehicleAbilitySystemComponent) "/Hoagie/GA_Hoagie_Explosion_Async.GA_Hoagie_Explosion_Async_C" self (Utilities|Struct|MakeGameplayEventData Instigator))
        (else
          (Class|VehicleFunctionLibrary|VehicleExplodeandDamageNearbyActors self Instigator (Variables|Default|GetExplodeRadius) (Variables|GameplayEffects|GetGE_ExplodeDamageVehicle) (Variables|GameplayEffects|GetGE_Explode_DamageBuilding) (Variables|GameplayEffects|GetGE_Explode_DamagePlayer) (Variables|Default|GetExplosionDamageVehicles) (Variables|Default|GetExplosionDamagePlayers) (Variables|Default|GetExplosionDamageBuildings) false true (Variables|Default|GetExplodeVehicleImpulse) (Variables|Default|GetExplodePlayerImpulse)))))))

(fn FindBoostCameraShake
  (bind (_returnvalue_x _returnvalue_y) (Math|Vector|BreakVector (Vehicle|GetThrustDirection false)))
  (bind _righthighvalue (Variables|Boost|GetRightHighValue))
  (bind _fowardhighvalue (Variables|Boost|GetFowardHighValue))
  (bind _returnvalue (>= _fowardhighvalue _righthighvalue))
  (bind _boostdeadzone (Variables|Boost|GetBoostDeadZone))
  (bind _returnvalue_2 (* _boostdeadzone -1.0))
  (Variables|Boost|SetBoostDirectionENUM "NewEnumerator0")
  (Variables|Boost|SetFowardHighValue 0.0)
  (Variables|Boost|SetRightHighValue 0.0)
  (if (>= _returnvalue_x _boostdeadzone)
    (Variables|Boost|SetFowardHighValue _returnvalue_x)
    (Variables|Boost|SetBoostDirectionENUM "NewEnumerator0")
    (elif (<= _returnvalue_x _returnvalue_2)
      (Variables|Boost|SetFowardHighValue (* _returnvalue_x -1.0))
      (Variables|Boost|SetBoostDirectionENUM "NewEnumerator1")
      (else
        _)))
  (if (>= _returnvalue_y _boostdeadzone)
    (Variables|Boost|SetRightHighValue _returnvalue_y)
    (if (not _returnvalue)
      (Variables|Boost|SetBoostDirectionENUM "NewEnumerator2"))
    (elif (<= _returnvalue_y _returnvalue_2)
      (Variables|Boost|SetRightHighValue (* _returnvalue_y -1.0))
      (if (not _returnvalue)
        (Variables|Boost|SetBoostDirectionENUM "NewEnumerator3"))))
  (return))

(fn FX_DestroyGlass
  (params (BoneSocketIndex :type Integer)))

(fn GetIsRotorWashActive
  (return (< (Variables|FX|RotorWash|Private|GetWashAltAlpha) 1.0)))

(fn GlassDestroyed
  (params (ShapeName :type Name))
  (switch name ShapeName
    (:window_l
      (CallFunction|SpawnGlassDestroyedFX 0)
      (Development|PrintString "LEFT" true true "(R=10.000000,G=0.166667,B=0.000000,A=1.000000)" 1e+01)
      (bind _audiocontroller (Utilities|GetAudioController)
        (:Success
          (Class|BPHoagieAudioController|OnNotifyGlassDestroyed _audiocontroller))))
    (:window_r
      (CallFunction|SpawnGlassDestroyedFX 1)
      (Development|PrintString "RIGHT" true true "(R=10.000000,G=0.166667,B=0.000000,A=1.000000)" 1e+01)
      (bind _audiocontroller_2 (Utilities|GetAudioController)
        (:Success
          (Class|BPHoagieAudioController|OnNotifyGlassDestroyed _audiocontroller_2))))
    (:window_c
      (CallFunction|SpawnGlassDestroyedFX 2)
      (Development|PrintString "FRONT" true true "(R=10.000000,G=0.166667,B=0.000000,A=1.000000)" 1e+01)
      (bind _audiocontroller_3 (Utilities|GetAudioController)
        (:Success
          (Class|BPHoagieAudioController|OnNotifyGlassDestroyed _audiocontroller_3))))))

(fn HandleCriticalDamage
  (params (Instigator :type Player%20Controller%20Object%20Reference) (Damage :type Float%20%28double-precision%29) (Origin :type Vector) (Damage%20Causer :type Actor%20Object%20Reference))
  (bind _asfort_player_controller_athena (Utilities|Casting|CastToFortPlayerControllerAthena Instigator))
  (bind _asfort_player_pawn (Utilities|Casting|CastToFortPlayerPawn (Pawn|GetControlledPawn _asfort_player_controller_athena)))
  (if (not (Team|OnSameTeam _asfort_player_pawn self))
    (bind _asfort_ability_system_component (Utilities|Casting|CastToFortAbilitySystemComponent (Ability|GetAbilitySystemComponent _asfort_player_pawn)))
    (bind _output_get (Variables|Default|SetCritContext (GameplayEffects|MakeFortEffectContext _asfort_ability_system_component Instigator 0 Damage%20Causer 0)))
    (Ability|EffectContext|SetCritical _output_get true)
    (Ability|EffectContext|SetOrigin _output_get Origin)
    (bind _returnvalue (Ability|GameplayEffect|AssignTagSetbyCallerMagnitude (GameplayEffects|MakeOutgoingSpec _asfort_ability_system_component (Variables|GameplayEffects|GetGE_WeakpointCritDamage) 0.0 _output_get) "(TagName=\"SetByCaller.AbilityDamage\")" (* Damage (Variables|Default|GetWeakpointDamageMultiplier))))
    (GameplayEffects|ApplyGameplayEffectSpecToTarget _asfort_ability_system_component _returnvalue (Vehicle|GetVehicleAbilitySystemComponent))))

(fn HoagieVehicle_AutoGenFunc
  (params (CriticalDamage :type Float%20%28double-precision%29) (DamageTags :type Gameplay%20Tag%20Container%20Structure) (HitInfo :type Hit%20Result%20Structure) (InstigatedBy :type Controller%20Object%20Reference) (DamageCauser :type Actor%20Object%20Reference) (EffectContext :type Gameplay%20Effect%20Context%20Handle%20Structure)))

(fn ManagePassengerEnter
  (params (PlayerPawn :type Fort%20Player%20Pawn%20Object%20Reference))
  (if (Pawn|IsLocallyControlled PlayerPawn)
    (Variables|Default|SetIsLocalPlayerPassenger true)
    (FortPawn|BindEventtoOnWeaponEquippedDelegate PlayerPawn (EventDispatchers|CreateEvent :!event WeaponEquipped))
    (Utilities|FlowControl|Sequence
      (:then_0
        (CallFunction|LocalPassengerMatSwap)
        (Collision|SetCollisionResponsetoChannel (Variables|Vehicle|GetSkeletalMesh) "ECC_Camera" "ECR_Ignore"))
      (:then_1
        (bind _asb_ranged_generic (Utilities|Casting|CastToB_Ranged_Generic (Variables|/Script/FortniteGame.FortPawn|GetCurrentWeapon PlayerPawn)))
        (CallFunction|WeaponEquipped _asb_ranged_generic)))))

(fn ManagePassengerExit
  (params (PlayerPawn :type Fort%20Player%20Pawn%20Object%20Reference))
  (if (Pawn|IsLocallyControlled PlayerPawn)
    (FortPawn|UnbindEventfromOnWeaponEquippedDelegate PlayerPawn (EventDispatchers|CreateEvent :!event WeaponEquipped))
    (bind _asb_ranged_generic (Utilities|Casting|CastToB_Ranged_Generic (Variables|/Script/FortniteGame.FortPawn|GetCurrentWeapon PlayerPawn))
      (:then
        (Default|UnbindEventfromOnAimDownSightsChanged _asb_ranged_generic (EventDispatchers|CreateEvent :!event LocalPassengerADS)))
      (:CastFailed
        _))
    (Variables|Default|SetIsLocalPlayerPassenger false)
    (CallFunction|LocalPassengerMatSwap)
    (Collision|SetCollisionResponsetoChannel (Variables|Vehicle|GetSkeletalMesh) "ECC_Camera" "ECR_Block")))

(fn ManagePilotEnter
  (params (Pilot :type Fort%20Player%20Pawn%20Object%20Reference))
  (Variables|Default|SetPilot Pilot)
  (Variables|Default|SetPilotController (Fortnite|GetFortPlayerControllerfromActor Pilot))
  (Utilities|FlowControl|Sequence
    (:then_0
      ;[1 1 1] Start Camera Shake
      (CallFunction|StartDriverCameraShake)
      (CallFunction|StartPassiveRumble)
      (if (not (Variables|Boost|GetIsBoosting))
        (Effects|Components|ParticleSystem|SetFloatParameter (Variables|Default|GetHoagie%20Idle%20FX) "User.BoostReady" 1.0)))
    (:then_1
      (Variables|Default|SetHasPilot true))))

(fn ManagePilotExit
  (bind _lastdamagecauser (Variables|Default|GetLastDamageCauser))
  (Utilities|FlowControl|Sequence
    (:then_0
      (bind _pilotcontroller (Variables|Default|GetPilotController)
        (:then
          (bind _playercameramanager (Variables|/Script/Engine.PlayerController|GetPlayerCameraManager _pilotcontroller)
            (:then
              (CameraShakes|StopAllInstancesOfCameraShake _playercameramanager (Variables|CameraShake|GetCameraShakeCritical) true))
            (:else
              _))
          (CallFunction|StopPilotCameraShake)
          (CallFunction|StopRotorHitCamera)
          (CallFunction|StopPassiveRumble)
          (CallFunction|StopBoostRumble)
          (CallFunction|CriticalRumbleStop)
          (CallFunction|StopSkidRumble)
          (bind _audiocontroller (Utilities|GetAudioController)
            (:Success
              (Class|BPHoagieAudioController|OnNotifyPilotExit _audiocontroller))))
        (:else
          _)))
    (:then_1
      (Effects|Components|ParticleSystem|SetFloatParameter (Variables|Default|GetHoagie%20Idle%20FX) "User.BoostReady" 0.0) ; turn off booster 
    )
    (:then_2
      (if (Vehicle|Destruction|IsCriticalHealth)
        (Utilities|IsValid _lastdamagecauser
          (:Is%20Valid
            (Class|FortPlayerPawnAthena|SetFallInstigator :self (Utilities|Casting|CastToFortPlayerPawnAthena (Variables|Default|GetPilot)) :NewFallInstigator (Game|GetInstigatorController _lastdamagecauser)))
          (:Is%20Not%20Valid
            (Variables|Default|SetPilot 0)))
        (Variables|Default|SetPilotController 0)
        (Variables|CameraShake|SetCriticalCameraShake 0)
        (Variables|Default|SetHasPilot false)
        (else
          _)))))

(fn OnRep_IsRotorWashActive
  (Networking|SwitchHasAuthority
    (:Remote
      (Niagara|SetNiagaraVariable(Bool) (Variables|Default|GetRotorWashFX) "User.isRotorWashActive" (Variables|Default|GetIsRotorWashActive)))))

(fn RemoveBoostCooldownEvent
  (params (Player%20Pawn :type Fort%20Player%20Pawn%20Object%20Reference))
  (bind _returnvalue (Ability|GetAbilitySystemComponent Player%20Pawn))
  (Ability|SendGameplayEventtoActor Player%20Pawn (Variables|Default|GetGamePlayEvent) (Utilities|Struct|MakeGameplayEventData :ContextHandle "()" :InstigatorTags "(GameplayTags=)" :TargetTags "(GameplayTags=)" :EventMagnitude 0.0 :TargetData "()"))
  (if (not (Class|ItemUIFunctionLibrary|IsNullItem _returnvalue))
    (GameplayEffects|RemoveActiveGameplayEffect _returnvalue (Variables|Default|GetCooldownCueEvent) -1)))

(fn SetDamageStateThresholds
  (bind _damage_thresholds (Variables|FX|DamageStage|GetDamage%20Thresholds))
  (bind _keys (Utilities|Map|Keys _damage_thresholds))
  (for _array_element _keys
    (if (.ReturnValue (Utilities|Map|Find _damage_thresholds _array_element))
      (Niagara|SetNiagaraVariable(Float) (Variables|Default|GetDamageFX) _array_element (Utilities|Map|Find _damage_thresholds _array_element))
      (else
        (Development|PrintString "Damage State Error - KEY MISSING")))))

(fn ShouldDamageBuilding
  (params (BuildingActor :type Building%20Actor%20Object%20Reference) (MaxHealth :type Float%20%28double-precision%29))
  (return (Math|Boolean|ANDBoolean (> (Health|GetMaxHealth BuildingActor :!class /Script/FortniteGame.BuildingActor) 0.0) (<= (Health|GetMaxHealth BuildingActor :!class /Script/FortniteGame.BuildingActor) MaxHealth) (not (Actor|ActorHasTag BuildingActor "NoVehicleDamage")) (Variables|/Script/Engine.Actor|GetbCanBeDamaged BuildingActor))))

(fn SpawnBooster
  (bind _returnvalue (Effects|SpawnFXSystemAttached :FXSystemAsset (Variables|FX|GetBoost%20Active%20Template) :AttachToComponent (Variables|Vehicle|GetSkeletalMesh) :AttachPointName "Hoagie_root" :Scale "1.000000,1.000000,1.000000" :LocationType "KeepRelativeOffset" :bAutoDestroy false :PoolingMethod "ManualRelease" :bAutoActivate true))
  (Variables|Default|SetBoostActive _returnvalue)
  (return))

(fn SpawnImpactEffect_Skid
  (params (Hit%20Location :type Vector) (Hit%20Normal :type Vector) (Impact%20Magnitude :type Float%20%28double-precision%29) (Physical%20Material :type EPhysicalSurface%20Enum :default "SurfaceType_Default"))
  (bind _returnvalue (Effects|Components|ParticleSystem|SpawnFXAtLocation (Variables|FX|Impact|GetNS_Hoagie_Scrape_LandingSkids) Hit%20Location))
  (Effects|Components|ParticleSystem|SetFloatParameter _returnvalue "User.Speed" (Vehicle|GetTotalSpeedKmh self))
  (Effects|Components|ParticleSystem|SetIntParameter _returnvalue "User.PhysMatInt" (select (Utilities|Enum|Equal(Enum) Physical%20Material "SurfaceType8") 1 0))
  (Effects|Components|ParticleSystem|SetVectorParameter _returnvalue "User.ThrustVelocity" (Math|Vector|RotateVectorAroundAxis (Vehicle|GetThrustDirection true) 1.8e+02 Hit%20Normal)))

(fn SpawnGlassDestroyedFX
  (params (Glass%20Destroyed :type Integer))
  (Networking|SwitchHasAuthority
    (:Remote
      (Effects|Components|ParticleSystem|SpawnFXAtLocation "/Hoagie/FX/NS_Hoagie_Window_Break_All.NS_Hoagie_Window_Break_All" (Math|Transform|TransformLocation (Transformation|GetActorTransform) (Utilities|Select :Index Glass%20Destroyed :Option%200 "160.000000,-68.000000,176.000000" :Option%201 "160.000000,68.000000,176.000000" :Option%202 "190.000000,0.000000,210.000000")) (Math|Rotator|CombineRotators (Math|Conversions|RotationFromXVector (Transformation|GetActorForwardVector)) (Utilities|Select :Index Glass%20Destroyed :Option%200 "0,-50.000000, 0" :Option%201 "0,50.000000, 0" :Option%202 "0, 0, 0")) "1.000000,1.000000,1.000000" true "AutoRelease")
      (return))))

(fn ConstructionScript
  (Variables|FortAthenaVehicle|SetOverlapComponent (Variables|Default|GetOverlapCollisionBox)))

(fn WeaponEquipped
  (params (NewWeapon :type Fort%20Weapon%20Object%20Reference) (PrevWeapon :type Fort%20Weapon%20Object%20Reference))
  (Utilities|FlowControl|Sequence
    (:then_0
      (bind _asb_ranged_generic (Utilities|Casting|CastToB_Ranged_Generic PrevWeapon))
      (Default|UnbindEventfromOnAimDownSightsChanged _asb_ranged_generic (EventDispatchers|CreateEvent :!event LocalPassengerADS)))
    (:then_1
      (bind _asb_ranged_generic_2 (Utilities|Casting|CastToB_Ranged_Generic NewWeapon))
      (Default|BindEventtoOnAimDownSightsChanged _asb_ranged_generic_2 (EventDispatchers|CreateEvent :!event LocalPassengerADS)))))
