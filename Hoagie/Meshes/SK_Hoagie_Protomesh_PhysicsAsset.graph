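# Physics asset for the Hoagie prototype skeletal mesh, giving it simplified box collision.
# A single body on the Hoagie_root bone combines two boxes: a wide, low Base and a smaller, taller Passengers box set higher, overlapping it.
# With no other bodies and no constraints, it defines no articulated parts or ragdoll joints.
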PhysicsAsset /Hoagie/Meshes/SK_Hoagie_Protomesh_PhysicsAsset
bodies {Hoagie_root [
  {shapeType=Box, extentX=387.375, extentY=373.6875, extentZ=60, shapeName=Base, center=<38.1919 0.0003 32.7478>}
  {shapeType=Box, extentX=200, extentY=317.1875, extentZ=150, shapeName=Passengers, center=<-10 0.0003 125>}
 ]}
constraints []
